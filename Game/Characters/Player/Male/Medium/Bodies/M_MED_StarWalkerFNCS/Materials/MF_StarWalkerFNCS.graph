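# Material function for the StarWalkerFNCS player body that overrides incoming material attributes with a black-base, animated galaxy emissive look.
# It sums panning cracks, a distorted pre-skinned-position inner glow, recolored nebula and two star layers, using rotated camera-vector triplanar projection.
# A curve-atlas LUT adds fresnel and optional sky-light specular, then time of day scales the sum.
# World position offset pushes camera-facing vertices toward the camera.
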MaterialFunction /Game/Characters/Player/Male/Medium/Bodies/M_MED_StarWalkerFNCS/Materials/MF_StarWalkerFNCS
graph {nodes [
  n1 = NamedRerouteUsage()
  n2 = ScalarParameter() {defaultValue=2.8616, parameterName=FallofGlass}
  n3 = PreSkinnedPosition()
  n4 = VectorParameter() {parameterName=Center, defaultValue=<rgba 0 0 126.2117 1>}
  n5 = Divide(
   A=Distance(A=ComponentMask(None=@n3) {b=true, r=true}, B=ComponentMask(None=@n4.RGB) {b=true, r=true})
   B=ScalarParameter() {defaultValue=80, parameterName=InnerGlowSize}
  )
  n6 = Add(
   A=Multiply(A=ScalarParameter() {defaultValue=0.5, parameterName=InnerGlowDistortion}, B=NamedRerouteUsage())
   B=ComponentMask(None=VertexInterpolator(VS=@n5).PS) {r=true}
  )
  n7 = Saturate(None=Power(Base=OneMinus(None=@n6), Exp=ScalarParameter() {defaultValue=8, parameterName=InnerGlowPower}))
  n8 = Add(
   A=Multiply(A=VectorParameter() {parameterName=PanningTextureColor, defaultValue=<rgba 1.6784 8 0 1>}.RGB, B=Power(Base=@n1, Exp=@n2))
   B=Multiply(
    A=NamedRerouteDeclaration(None=@n7) {name=Inner Glow, nodeColor=<rgba 0 1 0.1328 1>}
    B=VectorParameter() {parameterName=InnerGlowColor, defaultValue=<rgba 0.3005 1 0.5583 1>}.RGB
   )
  )
  n9 = Multiply(A=Power(Base=Fresnel() {baseReflectFraction=0.01, exponent=0.2}) {constExponent=1})
  n10 = ScalarParameter() {defaultValue=5, parameterName=GalaxyIntensity}
  n11 = ScalarParameter() {defaultValue=0.2314, parameterName=GalaxyColorFresnelBlend}
  n12 = NamedRerouteUsage()
  n13 = Multiply(
   A=LinearInterpolate(
    A=VectorParameter() {parameterName=GalaxyColor1, defaultValue=<rgba 1 0.6977 0.6927 1>}.RGB
    B=VectorParameter() {parameterName=GalaxyColor2, defaultValue=<rgba 0.7748 2 0.1667 1>}.RGB
    Alpha=Fresnel(ExponentIn=@n11) {baseReflectFraction=0}
   )
   B=ComponentMask(None=@n12) {r=true}
  )
  n14 = Multiply(A=ComponentMask(None=@n12) {b=true}, B=VectorParameter() {parameterName=GalaxyColor3, defaultValue=<rgba 0.833 0.2833 0.3918 1>}.RGB)
  n15 = LinearInterpolate(
   A=Multiply(A=Multiply(A=@n9, B=@n10), B=LinearInterpolate(A=@n13, B=@n14))
   B=NamedRerouteUsage()
   Alpha=ScalarParameter() {defaultValue=0.0762, parameterName=GalaxyOriginalTextureBlend}
  )
  n16 = VectorParameter() {parameterName=RiftShardColor, defaultValue=<rgba 0.9279 7.6847 10 0>}
  n17 = VectorParameter() {group=Galaxy, parameterName=Star Brightness_1, defaultValue=<rgba 10 10 10 1>}
  n18 = Time()
  n19 = ScalarParameter() {defaultValue=0.0156, parameterName=Time Scale}
  n20 = TextureObject() {samplerType=SAMPLERTYPE_Masks, texture=/Game/Characters/Player/Male/Medium/Bodies/M_Med_Soldier_04/Skins/Galaxy/Textures/Celestial_T_Stars}
  n21 = ComponentMask(None=MaterialFunctionCall!(
    CameraVector
    Rotation Speed=Multiply(A=@n18, B=@n19)
    Texture=@n20
    Tiling=ScalarParameter() {defaultValue=4, group=Galaxy, parameterName=MediumStarTiling}
    Offset=ScalarParameter() {defaultValue=0.012, group=Galaxy, parameterName=MediumStar_WarpIntensity}
    AxisFadeContrast=ScalarParameter() {defaultValue=6, parameterName=AxisFadeContrast}
   ) {materialFunction=/Game/Weapons/WeaponSkins/MaterialFunction/TriplanarCameraVector_Rotated}.Result) {b=true}
  n22 = TextureSample(UVs=Panner(Coordinate=TextureCoordinate() {uTiling=2, vTiling=2}) {speedX=0.137, speedY=-0.2}) {
   samplerSource=SSM_Wrap_WorldGroupSettings
   samplerType=SAMPLERTYPE_Masks
   texture=&Celestial_T_Stars
  }
  n23 = Multiply(A=Multiply(A=@n21, B=@n21), B=@n22.G)
  n24 = VectorParameter() {group=Galaxy, parameterName=Star Brightness_2, defaultValue=<rgba 10 10 10 20>}
  n25 = Time()
  n26 = ScalarParameter() {defaultValue=0.0156, parameterName=Time Scale}
  n27 = TextureObject() {samplerType=SAMPLERTYPE_Masks, texture=&Celestial_T_Stars}
  n28 = ComponentMask(None=MaterialFunctionCall(
    Rotation Speed=Multiply(A=@n25, B=@n26)
    Texture=@n27
    Tiling=ScalarParameter() {defaultValue=2, group=Galaxy, parameterName=LargeStarTiling}
    Offset=ScalarParameter() {defaultValue=0.012, group=Galaxy, parameterName=LargeStar_WarpIntensity}
    AxisFadeContrast=ScalarParameter() {defaultValue=6, parameterName=AxisFadeContrast}
   ) {materialFunction=&TriplanarCameraVector_Rotated}.Result) {r=true}
  n29 = TextureSample(UVs=Panner(Coordinate=TextureCoordinate()) {speedX=0.137, speedY=-0.2}) {
   samplerSource=SSM_Wrap_WorldGroupSettings
   samplerType=SAMPLERTYPE_Masks
   texture=&Celestial_T_Stars
  }
  n30 = Multiply(A=Multiply(A=@n28, B=@n28), B=@n29.G)
  n31 = Multiply(A=NamedRerouteUsage(), B=Add(A=Multiply(A=@n17.RGB, B=@n23), B=Multiply(A=@n24.RGB, B=@n30)))
  n32 = Fresnel(ExponentIn=ScalarParameter() {defaultValue=1, parameterName=FresnelPower}) {baseReflectFraction=0.01}
  n33 = NamedRerouteUsage()
  n34 = Add(
   A=Multiply(A=Power(Base=@n32) {constExponent=5}, B=ScalarParameter() {defaultValue=5, parameterName=FresnelMultiplier}) {constB=5}
   B=Multiply(A=Saturate(None=@n32), B=Desaturation(None=@n33))
  )
  n35 = MaterialFunctionCall!(Normal=PixelNormalWS(), CameraVector=SkyAtmosphereLightDirection()) {materialFunction=/Engine/Functions/Engine_MaterialFunctions02/Utility/CustomReflectionVector}
  n36 = CameraVectorWS()
  n37 = Power(Base=Saturate(None=DotProduct(A=@n35.Reflection Vector, B=@n36)), Exp=ScalarParameter() {defaultValue=32, parameterName=SpecularPower}) {constExponent=12}
  n38 = Add(A=@n34, B=Multiply(A=Saturate(None=@n37), B=ScalarParameter() {defaultValue=1, parameterName=SpecularMultiplier}) {constB=8})
  n39 = OneMinus(None=NamedRerouteUsage())
  n40 = CurveAtlasRowParameter(CurveTime=StaticSwitchParameter(True=Multiply(A=@n38, B=@n39), False=@n34) {defaultValue=true, parameterName=Enable Specular}) {
   atlas=&/CA_StarWalkFNCS
   curve=&/C_StarWalkerFNCS
   parameterName=LUT_CurveTexture
  }
  n41 = Add(A=Add(A=Add(A=@n8, B=@n15), B=Multiply(A=@n16.RGB, B=@n31)), B=@n40)
  n42 = Multiply(A=@n41) {constB=0.4}
  n43 = MaterialFunctionCall(Color_Morning=@n41, Color_Day=@n41, Color_Evening=@n42, Color_Night=@n42) {materialFunction=&Color_Time_of_Day}
  n44 = ScalarParameter() {defaultValue=1, parameterName=Mobile Emissive Scale}
  n45 = Normalize(VectorInput=Subtract(A=CameraPositionWS(), B=WorldPosition().XYZ))
  n46 = ScalarParameter() {defaultValue=5, parameterName=PushtoCamera}
  n47 = Power(Base=DotProduct(A=VertexNormalWS(), B=@n45), Exp=ScalarParameter() {defaultValue=0.7, parameterName=WPOFrenellPower})
  n48 = SetMaterialAttributes!(
   MaterialAttributes=FunctionInput() {inputType=FunctionInput_MaterialAttributes}
   BaseColor=Constant()
   EmissiveColor=ShadingPathSwitch(Default=@n43.Result, Mobile=Multiply(A=@n43.Result, B=@n44))
   WorldPositionOffset=LinearInterpolate(A=Constant(), B=Multiply(A=@n45, B=@n46), Alpha=Saturate(None=@n47))
   Metallic=Constant()
   Roughness=Constant() {r=0.5}
  ) {attributeSetTypes [
    #1
    #2
    #3
    #4
    #5
   ]}
  FunctionOutput(None=@n48)
  n49 = Multiply(A=Time(), B=VectorParameter() {parameterName=Panner SpeedG, defaultValue=<rgba 0.0772 0.0772 0.0772 1>}.RGB)
  n50 = TextureObjectParameter() {parameterName=Panning Texture Mask, texture=/Game/Effects/Fort_Effects/Textures/Tile/T_Glass_Cracks_M}
  n51 = LinearInterpolate(
   A=ScalarParameter() {defaultValue=2, group=Galaxy, parameterName=GalaxyTiling_NearG}
   B=ScalarParameter() {defaultValue=4, group=Galaxy, parameterName=GalaxyTiling_FarG}
   Alpha=MaterialFunctionCall(Distance=ScalarParameter() {defaultValue=2000, parameterName=Near Camera Tiling DistanceG}) {materialFunction=&Near_Camera_Fade}.Result
  )
  n52 = ScalarParameter() {defaultValue=5, group=Galaxy, parameterName=WarpIntensityG}
  n53 = ScalarParameter() {defaultValue=3, parameterName=AxisFadeContrastG}
  NamedRerouteDeclaration(None=ComponentMask(None=MaterialFunctionCall(Rotation Speed=@n49, Texture=@n50, Tiling=@n51, Offset=@n52, AxisFadeContrast=@n53) {materialFunction=&TriplanarCameraVector_Rotated}.Result) {r=true}) {name=Panning Texture, nodeColor=<rgba 1 0.3281 0 1>}
  n54 = Multiply(
   A=ScalarParameter() {defaultValue=1.0039, parameterName=Distort}
   B=Multiply(A=Constant3Vector() {constant=<rgba 1 1 1 0>}, B=NamedRerouteUsage())
  )
  n55 = CameraVectorWS()
  n56 = Time()
  n57 = VectorParameter() {parameterName=Panner Speed, defaultValue=<rgba 0.125 0.125 -0.0625 1>}
  n58 = ScalarParameter() {defaultValue=2, group=Galaxy, parameterName=GalaxyTiling_Near}
  n59 = ScalarParameter() {defaultValue=1, group=Galaxy, parameterName=GalaxyTiling_Far}
  n60 = MaterialFunctionCall(Distance=ScalarParameter() {defaultValue=1000, parameterName=Near Camera Tiling Distance}) {materialFunction=&Near_Camera_Fade}
  NamedRerouteDeclaration(None=MaterialFunctionCall(
    CameraVector=Add(A=@n54, B=@n55)
    Rotation Speed=Multiply(A=@n56, B=@n57.RGB)
    Texture=TextureObjectParameter() {parameterName=Galaxy Texture, texture=/Game/Characters/Player/Male/Medium/Bodies/M_Med_Soldier_04/Skins/Galaxy/Textures/T_FN_Nebula}
    Tiling=LinearInterpolate(A=@n58, B=@n59, Alpha=@n60.Result)
    Offset=ScalarParameter() {defaultValue=0.012, group=Galaxy, parameterName=WarpIntensity}
    AxisFadeContrast=ScalarParameter() {defaultValue=6, parameterName=AxisFadeContrast}
   ) {materialFunction=&TriplanarCameraVector_Rotated}.Result) {name=Galaxy Background, nodeColor=<rgba 0.2891 0 1 1>}
  VertexNormalWS()
 ]}
