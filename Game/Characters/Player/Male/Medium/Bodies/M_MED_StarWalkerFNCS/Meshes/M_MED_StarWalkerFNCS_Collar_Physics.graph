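# Physics asset for the collar of the male medium StarWalkerFNCS character, defining dynamics bodies for the cowl and left shoulder tips.
# Each side's cowl is a two-box chain and each of the two left shoulder tips is a single box, all jointed with swing and twist limits.
# Small sphere parent bodies root these chains to spine_05 through angularly free constraints, and two capsules on spine_05 form the upper-torso volume.
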PhysicsAsset /Game/Characters/Player/Male/Medium/Bodies/M_MED_StarWalkerFNCS/Meshes/M_MED_StarWalkerFNCS_Collar_Physics
bodies {
 dyn_cowl_01_l [{shapeType=Box, extentX=11.5, extentY=2.5, extentZ=12, center=<4 0 0>}]
 dyn_cowl_01_r [{shapeType=Box, extentX=11.5, extentY=2.5, extentZ=12, center=<4 0 0>}]
 dyn_cowl_02_l [{shapeType=Box, extentX=9.5, extentY=2.5, extentZ=13.5, center=<4 -0.3616 0.0644>, rotation=<rot 0 0 10.095>}]
 dyn_cowl_02_r [{shapeType=Box, extentX=6.5, extentY=2.5, extentZ=6.5, center=<4 -0.5 0.0523>, rotation=<rot -0.0001 -0.0002 12.1738>}]
 dyn_parent_cowl_l [{shapeType=Sphere, radius=0.505}]
 dyn_parent_cowl_r [{shapeType=Sphere, radius=0.505}]
 dyn_parent_shldr_tipA_l [{shapeType=Sphere, radius=0.505}]
 dyn_parent_shldr_tipB_l [{shapeType=Sphere, radius=0.505}]
 dyn_shldr_tipA_01_l [{shapeType=Box, extentX=6, extentY=1, extentZ=4, center=<3 0 0>}]
 dyn_shldr_tipB_01_l [{shapeType=Box, extentX=6, extentY=1, extentZ=4, center=<3 0 0>}]
 spine_05 [
  {shapeType=Capsule, length=12, radius=12, center=<-8 2 0>}
  {shapeType=Capsule, length=13.25, radius=19.7607, center=<-13.5688 -6.8856 -6.4392>, rotation=<rot -40.8662 -88.6201 -92.3421>}
 ]
}
constraints [
 {bone1Name=dyn_cowl_01_l, bone2Name=dyn_parent_cowl_l, swing1LimitDegrees=45, swing2LimitDegrees=45, twistLimitDegrees=15}
 {bone1Name=dyn_cowl_02_l, bone2Name=dyn_cowl_01_l, swing1LimitDegrees=45, swing2LimitDegrees=45, twistLimitDegrees=15}
 {bone1Name=dyn_cowl_01_r, bone2Name=dyn_parent_cowl_r, swing1LimitDegrees=45, swing2LimitDegrees=45, twistLimitDegrees=15}
 {bone1Name=dyn_cowl_02_r, bone2Name=dyn_cowl_01_r, swing1LimitDegrees=45, swing2LimitDegrees=45, twistLimitDegrees=15}
 {bone1Name=dyn_shldr_tipA_01_l, bone2Name=dyn_parent_shldr_tipA_l, swing1LimitDegrees=45, swing2LimitDegrees=45, twistLimitDegrees=45}
 {bone1Name=dyn_shldr_tipB_01_l, bone2Name=dyn_parent_shldr_tipB_l, swing1LimitDegrees=45, swing2LimitDegrees=45, twistLimitDegrees=45}
 {
  bone1Name=dyn_parent_shldr_tipA_l
  bone2Name=spine_05
  swing1LimitDegrees=45
  swing1Motion=Free
  swing2LimitDegrees=45
  swing2Motion=Free
  twistLimitDegrees=45
  twistMotion=Free
 }
 {
  bone1Name=dyn_parent_shldr_tipB_l
  bone2Name=spine_05
  swing1LimitDegrees=45
  swing1Motion=Free
  swing2LimitDegrees=45
  swing2Motion=Free
  twistLimitDegrees=45
  twistMotion=Free
 }
 {
  bone1Name=dyn_parent_cowl_l
  bone2Name=spine_05
  swing1LimitDegrees=45
  swing1Motion=Free
  swing2LimitDegrees=45
  swing2Motion=Free
  twistLimitDegrees=45
  twistMotion=Free
 }
 {
  bone1Name=dyn_parent_cowl_r
  bone2Name=spine_05
  swing1LimitDegrees=45
  swing1Motion=Free
  swing2LimitDegrees=45
  swing2Motion=Free
  twistLimitDegrees=45
  twistMotion=Free
 }
]
