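# Control Rig for the StarWalkerFNCS medium male body that applies corrective rotations to jacket bones as the character poses.
# Spherical pose readers on both thighs weight rotation offsets on front and side jacket bones.
# Right upper arm readers, multiplied by crouch and rifle or pistol alphas, adjust left front jacket bones.
# The last two thigh reader groups apply zero rotation, and the spine_05 group is disconnected from execution.
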ControlRigBlueprint /Game/Characters/Player/Male/Medium/Bodies/M_MED_StarWalkerFNCS/Meshes/M_MED_StarWalkerFNCS_CtrlRig
graphs [
 Multiply_1_1_1_ContainedGraph {nodes [
   n1 = RigVMFunctionEntryNode()
   RigVMFunctionReturnNode(Result=&RigVMFunction_MathDoubleMul(B=@n1.C, A=&RigVMFunction_MathDoubleMul(A=@n1.A, B=@n1.B).Result).Result)
  ]}
 Multiply_1_ContainedGraph {nodes [
   n1 = RigVMFunctionEntryNode()
   Multiply = &RigVMFunction_MathDoubleMul(A=@n1.A, B=@n1.B)
   RigVMFunctionReturnNode(Result=&RigVMFunction_MathDoubleMul(B=@n1.C, A=@Multiply.Result).Result)
  ]}
 RigVMFunctionLibrary {links [], nodes []}
 RigVMModel {nodes [
   RigUnit_BeginExecution = &RigUnit_BeginExecution()
   n1 = &RigUnit_SphericalPoseReader(
    ActiveRegionScaleFactors=(PositiveWidth=3.5,NegativeWidth=0.1,PositiveHeight=3,NegativeHeight=3)
    ActiveRegionSize=0.1
    Debug=(bDrawDebug=false,bDraw2D=false,bDrawLocalAxes=false,DebugScale=25,DebugSegments=20,DebugThickness=0.25)
    DriverAxis=(X=-1,Y=0,Z=0)
    DriverItem=(Type=Bone,Name="thigh_l")
    FalloffRegionScaleFactors=(PositiveWidth=1,NegativeWidth=1.1,PositiveHeight=1,NegativeHeight=1)
    FalloffSize=0.45
    OptionalParentItem=(Type=Bone,Name="pelvis")
    RotationOffset=(X=120,Y=0,Z=-17)
    ExecutePin=@RigUnit_BeginExecution.ExecutePin
   )
   > &RigUnit_ModifyTransforms(
    ItemToModify=((Item=(Type=Bone,Name="jacket_frt_01_l"),Transform=(Rotation=(X=0,Y=0,Z=0,W=1),Translation=(X=0,Y=0,Z=0),Scale3D=(X=1,Y=1,Z=1))))
    ItemToModify.0.Transform.Rotation=&RigVMFunction_MathQuaternionFromEuler(Euler=(X=0,Y=-10,Z=-10), RotationOrder=XYZ).Result
    Weight=@n1.OutputParam
   )
   > &RigUnit_ModifyTransforms(
    ItemToModify=((Item=(Type=Bone,Name="jacket_frt_02_l"),Transform=(Rotation=(X=0,Y=0,Z=0,W=1),Translation=(X=0,Y=0,Z=0),Scale3D=(X=1,Y=1,Z=1))))
    ItemToModify.0.Transform.Rotation=&RigVMFunction_MathQuaternionFromEuler(Euler=(X=0,Y=0,Z=-10), RotationOrder=XYZ).Result
    Weight=@n1.OutputParam
   )
   > &RigUnit_ModifyTransforms(
    ItemToModify=((Item=(Type=Bone,Name="jacket_frtSide_01_l"),Transform=(Rotation=(X=0,Y=0,Z=0,W=1),Translation=(X=0,Y=0,Z=0),Scale3D=(X=1,Y=1,Z=1))))
    ItemToModify.0.Transform.Rotation=&RigVMFunction_MathQuaternionFromEuler(Euler=(X=0,Y=-10,Z=-10), RotationOrder=XYZ).Result
    Weight=@n1.OutputParam
   )
   > &RigUnit_ModifyTransforms(
    ItemToModify=((Item=(Type=Bone,Name="jacket_side_01_l"),Transform=(Rotation=(X=0,Y=0,Z=0,W=1),Translation=(X=0,Y=0,Z=0),Scale3D=(X=1,Y=1,Z=1))))
    ItemToModify.0.Transform.Rotation=&RigVMFunction_MathQuaternionFromEuler(Euler=(X=0,Y=0,Z=-5), RotationOrder=XYZ).Result
    Weight=@n1.OutputParam
   )
   > &RigUnit_ModifyTransforms(
    ItemToModify=((Item=(Type=Bone,Name="jacket_frt_04_l"),Transform=(Rotation=(X=0,Y=0,Z=0,W=1),Translation=(X=0,Y=0,Z=0),Scale3D=(X=1,Y=1,Z=1))))
    ItemToModify.0.Transform.Rotation=&RigVMFunction_MathQuaternionFromEuler(Euler=(X=0,Y=-20,Z=0), RotationOrder=XYZ).Result
    Weight=@n1.OutputParam
   )
   > n2 = &RigUnit_SphericalPoseReader(
    ActiveRegionScaleFactors=(PositiveWidth=3.5,NegativeWidth=0.1,PositiveHeight=3,NegativeHeight=3)
    ActiveRegionSize=0.1
    Debug=(bDrawDebug=false,bDraw2D=false,DebugScale=25,DebugSegments=20,DebugThickness=0.25)
    DriverAxis=(X=1,Y=0,Z=0)
    DriverItem=(Type=Bone,Name="thigh_r")
    FalloffRegionScaleFactors=(PositiveWidth=1,NegativeWidth=1.1,PositiveHeight=1,NegativeHeight=1)
    FalloffSize=0.45
    OptionalParentItem=(Type=Bone,Name="pelvis")
    RotationOffset=(X=-120,Y=180,Z=-17)
   )
   > &RigUnit_ModifyTransforms(
    ItemToModify=((Item=(Type=Bone,Name="jacket_frt_01_r"),Transform=(Rotation=(X=0,Y=0,Z=0,W=1),Translation=(X=0,Y=0,Z=0),Scale3D=(X=1,Y=1,Z=1))))
    ItemToModify.0.Transform.Rotation=&RigVMFunction_MathQuaternionFromEuler(Euler=(X=0,Y=10,Z=-10), RotationOrder=XYZ).Result
    Weight=@n2.OutputParam
   )
   > &RigUnit_ModifyTransforms(
    ItemToModify=((Item=(Type=Bone,Name="jacket_frt_02_r"),Transform=(Rotation=(X=0,Y=0,Z=0,W=1),Translation=(X=0,Y=0,Z=0),Scale3D=(X=1,Y=1,Z=1))))
    ItemToModify.0.Transform.Rotation=&RigVMFunction_MathQuaternionFromEuler(Euler=(X=0,Y=0,Z=-10), RotationOrder=XYZ).Result
    Weight=@n2.OutputParam
   )
   > &RigUnit_ModifyTransforms(
    ItemToModify=((Item=(Type=Bone,Name="jacket_frtSide_01_r"),Transform=(Rotation=(X=0,Y=0,Z=0,W=1),Translation=(X=0,Y=0,Z=0),Scale3D=(X=1,Y=1,Z=1))))
    ItemToModify.0.Transform.Rotation=&RigVMFunction_MathQuaternionFromEuler(Euler=(X=0,Y=10,Z=-10), RotationOrder=XYZ).Result
    Weight=@n2.OutputParam
   )
   > &RigUnit_ModifyTransforms(
    ItemToModify=((Item=(Type=Bone,Name="jacket_side_01_r"),Transform=(Rotation=(X=0,Y=0,Z=0,W=1),Translation=(X=0,Y=0,Z=0),Scale3D=(X=1,Y=1,Z=1))))
    ItemToModify.0.Transform.Rotation=&RigVMFunction_MathQuaternionFromEuler(Euler=(X=0,Y=0,Z=-5), RotationOrder=XYZ).Result
    Weight=@n2.OutputParam
   )
   > &RigUnit_ModifyTransforms(
    ItemToModify=((Item=(Type=Bone,Name="jacket_frt_04_r"),Transform=(Rotation=(X=0,Y=0,Z=0,W=1),Translation=(X=0,Y=0,Z=0),Scale3D=(X=1,Y=1,Z=1))))
    ItemToModify.0.Transform.Rotation=&RigVMFunction_MathQuaternionFromEuler(Euler=(X=0,Y=20,Z=0), RotationOrder=XYZ).Result
    Weight=@n2.OutputParam
   )
   > n3 = &RigUnit_SphericalPoseReader(
    ActiveRegionScaleFactors=(PositiveWidth=0.5,NegativeWidth=0.5,PositiveHeight=0.5,NegativeHeight=0.5)
    ActiveRegionSize=0.1
    Debug=(bDrawDebug=false,bDraw2D=false,bDrawLocalAxes=false,DebugScale=25,DebugSegments=20,DebugThickness=0.25)
    DriverAxis=(X=-1,Y=0,Z=0)
    DriverItem=(Type=Bone,Name="upperarm_r")
    FalloffRegionScaleFactors=(PositiveWidth=1,NegativeWidth=1,PositiveHeight=1,NegativeHeight=1)
    FalloffSize=0.25
    OptionalParentItem=(Type=Bone,Name="clavicle_r")
    RotationOffset=(X=-22,Y=27,Z=0)
   )
   Multiply_1_1 = RigVMAggregateNode(A=@n3.OutputParam, B=RigVMVariableNode() {variable=Is Rifle Alpha}.Value, C=RigVMVariableNode() {variable=Is Crouching Float}.Value)
   > &RigUnit_ModifyTransforms(
    ItemToModify=((Item=(Type=Bone,Name="jacket_frt_02_l"),Transform=(Rotation=(X=0,Y=0,Z=0,W=1),Translation=(X=0,Y=0,Z=0),Scale3D=(X=1,Y=1,Z=1))))
    ItemToModify.0.Transform.Rotation=&RigVMFunction_MathQuaternionFromEuler(Euler=(X=0,Y=10,Z=-25)).Result
    Weight=@Multiply_1_1.Result
   )
   > n4 = &RigUnit_SphericalPoseReader(
    ActiveRegionScaleFactors=(PositiveWidth=0.5,NegativeWidth=0.5,PositiveHeight=2,NegativeHeight=0)
    ActiveRegionSize=0.1
    Debug=(bDrawDebug=false,bDraw2D=false,bDrawLocalAxes=false,DebugScale=25,DebugSegments=20,DebugThickness=0.25)
    DriverAxis=(X=-1,Y=0,Z=0)
    DriverItem=(Type=Bone,Name="upperarm_r")
    FalloffRegionScaleFactors=(PositiveWidth=0.25,NegativeWidth=0.25,PositiveHeight=0.5,NegativeHeight=0.5)
    FalloffSize=0.45
    OptionalParentItem=(Type=Bone,Name="clavicle_r")
    RotationOffset=(X=-65,Y=20,Z=15)
   )
   Multiply = RigVMAggregateNode(A=@n4.OutputParam, C=RigVMVariableNode() {variable=Is Crouching Float}.Value, B=RigVMVariableNode() {variable=Is Pistol Alpha}.Value)
   > &RigUnit_ModifyTransforms(
    ItemToModify=((Item=(Type=Bone,Name="jacket_frt_01_l"),Transform=(Rotation=(X=0,Y=0,Z=0,W=1),Translation=(X=0,Y=0,Z=0),Scale3D=(X=1,Y=1,Z=1))))
    ItemToModify.0.Transform.Rotation=&RigVMFunction_MathQuaternionFromEuler(Euler=(X=0,Y=0,Z=-20)).Result
    Weight=@Multiply.Result
   )
   > n5 = &RigUnit_SphericalPoseReader(
    ActiveRegionScaleFactors=(PositiveWidth=3.5,NegativeWidth=1.5,PositiveHeight=2,NegativeHeight=2)
    ActiveRegionSize=0.1
    Debug=(bDrawDebug=false,bDraw2D=false,DebugScale=25,DebugSegments=20,DebugThickness=0.25)
    DriverAxis=(X=-1,Y=0,Z=0)
    DriverItem=(Type=Bone,Name="thigh_l")
    FalloffRegionScaleFactors=(PositiveWidth=1,NegativeWidth=1,PositiveHeight=0.5,NegativeHeight=0.5)
    FalloffSize=0.45
    OptionalParentItem=(Type=Bone,Name="pelvis")
    RotationOffset=(X=150,Y=0,Z=0)
   )
   > &RigUnit_ModifyTransforms(
    ItemToModify=((Item=(Type=Bone,Name="jacket_frt_01_l"),Transform=(Rotation=(X=0,Y=0,Z=0,W=1),Translation=(X=0,Y=0,Z=0),Scale3D=(X=1,Y=1,Z=1))))
    ItemToModify.0.Transform.Rotation=&RigVMFunction_MathQuaternionFromEuler(Euler=(X=0,Y=0,Z=-10), RotationOrder=XYZ).Result
    Weight=@n5.OutputParam
   )
   > &RigUnit_ModifyTransforms(
    ItemToModify=((Item=(Type=Bone,Name="jacket_frtSide_01_l"),Transform=(Rotation=(X=0,Y=0,Z=0,W=1),Translation=(X=0,Y=0,Z=0),Scale3D=(X=1,Y=1,Z=1))))
    ItemToModify.0.Transform.Rotation=&RigVMFunction_MathQuaternionFromEuler(Euler=(X=0,Y=0,Z=-5), RotationOrder=XYZ).Result
    Weight=@n5.OutputParam
   )
   > &RigUnit_ModifyTransforms(
    ItemToModify=((Item=(Type=Bone,Name="jacket_side_01_l"),Transform=(Rotation=(X=0,Y=0,Z=0,W=1),Translation=(X=0,Y=0,Z=0),Scale3D=(X=1,Y=1,Z=1))))
    ItemToModify.0.Transform.Rotation=&RigVMFunction_MathQuaternionFromEuler(Euler=(X=0,Y=0,Z=-5), RotationOrder=XYZ).Result
    Weight=@n5.OutputParam
   )
   > n6 = &RigUnit_SphericalPoseReader(
    ActiveRegionScaleFactors=(PositiveWidth=3.5,NegativeWidth=1.5,PositiveHeight=2,NegativeHeight=2)
    ActiveRegionSize=0.1
    Debug=(bDrawDebug=false,bDraw2D=false,DebugScale=25,DebugSegments=20,DebugThickness=0.25)
    DriverAxis=(X=1,Y=0,Z=0)
    DriverItem=(Type=Bone,Name="thigh_r")
    FalloffRegionScaleFactors=(PositiveWidth=1,NegativeWidth=1,PositiveHeight=0.5,NegativeHeight=0.5)
    FalloffSize=0.45
    OptionalParentItem=(Type=Bone,Name="pelvis")
    RotationOffset=(X=-150,Y=180,Z=0)
   )
   > &RigUnit_ModifyTransforms(
    ItemToModify=((Item=(Type=Bone,Name="jacket_frt_01_r"),Transform=(Rotation=(X=0,Y=0,Z=0,W=1),Translation=(X=0,Y=0,Z=0),Scale3D=(X=1,Y=1,Z=1))))
    ItemToModify.0.Transform.Rotation=&RigVMFunction_MathQuaternionFromEuler(Euler=(X=0,Y=0,Z=-10), RotationOrder=XYZ).Result
    Weight=@n6.OutputParam
   )
   > &RigUnit_ModifyTransforms(
    ItemToModify=((Item=(Type=Bone,Name="jacket_frtSide_01_r"),Transform=(Rotation=(X=0,Y=0,Z=0,W=1),Translation=(X=0,Y=0,Z=0),Scale3D=(X=1,Y=1,Z=1))))
    ItemToModify.0.Transform.Rotation=&RigVMFunction_MathQuaternionFromEuler(Euler=(X=0,Y=0,Z=-5), RotationOrder=XYZ).Result
    Weight=@n6.OutputParam
   )
   > &RigUnit_ModifyTransforms(
    ItemToModify=((Item=(Type=Bone,Name="jacket_side_01_r"),Transform=(Rotation=(X=0,Y=0,Z=0,W=1),Translation=(X=0,Y=0,Z=0),Scale3D=(X=1,Y=1,Z=1))))
    ItemToModify.0.Transform.Rotation=&RigVMFunction_MathQuaternionFromEuler(Euler=(X=0,Y=0,Z=-5), RotationOrder=XYZ).Result
    Weight=@n6.OutputParam
   )
   > n7 = &RigUnit_SphericalPoseReader(
    ActiveRegionScaleFactors=(PositiveWidth=2,NegativeWidth=1.5,PositiveHeight=2,NegativeHeight=2)
    ActiveRegionSize=0.1
    Debug=(bDrawDebug=false,bDraw2D=false,DebugScale=25,DebugSegments=20,DebugThickness=0.25)
    DriverAxis=(X=-1,Y=0,Z=0)
    DriverItem=(Type=Bone,Name="thigh_l")
    FalloffRegionScaleFactors=(PositiveWidth=1,NegativeWidth=1.25,PositiveHeight=1,NegativeHeight=1)
    FalloffSize=0.45
    OptionalParentItem=(Type=Bone,Name="pelvis")
    RotationOffset=(X=150,Y=180,Z=10)
   )
   > &RigUnit_ModifyTransforms(
    ItemToModify=((Item=(Type=Bone,Name="jacket_frt_02_l"),Transform=(Rotation=(X=0,Y=0,Z=0,W=1),Translation=(X=0,Y=0,Z=0),Scale3D=(X=1,Y=1,Z=1))))
    ItemToModify.0.Transform.Rotation=&RigVMFunction_MathQuaternionFromEuler(Euler=(X=0,Y=0,Z=0), RotationOrder=XYZ).Result
    Weight=@n7.OutputParam
   )
   > &RigUnit_ModifyTransforms(
    ItemToModify=((Item=(Type=Bone,Name="jacket_frtSide_01_l"),Transform=(Rotation=(X=0,Y=0,Z=0,W=1),Translation=(X=0,Y=0,Z=0),Scale3D=(X=1,Y=1,Z=1))))
    ItemToModify.0.Transform.Rotation=&RigVMFunction_MathQuaternionFromEuler(Euler=(X=0,Y=0,Z=0), RotationOrder=XYZ).Result
    Weight=@n7.OutputParam
   )
   > &RigUnit_ModifyTransforms(
    ItemToModify=((Item=(Type=Bone,Name="jacket_side_01_l"),Transform=(Rotation=(X=0,Y=0,Z=0,W=1),Translation=(X=0,Y=0,Z=0),Scale3D=(X=1,Y=1,Z=1))))
    ItemToModify.0.Transform.Rotation=&RigVMFunction_MathQuaternionFromEuler(Euler=(X=0,Y=0,Z=0), RotationOrder=XYZ).Result
    Weight=@n7.OutputParam
   )
   > n8 = &RigUnit_SphericalPoseReader(
    ActiveRegionScaleFactors=(PositiveWidth=2,NegativeWidth=1.5,PositiveHeight=2,NegativeHeight=2)
    ActiveRegionSize=0.1
    Debug=(bDrawDebug=false,bDraw2D=false,DebugScale=25,DebugSegments=20,DebugThickness=0.25)
    DriverAxis=(X=1,Y=0,Z=0)
    DriverItem=(Type=Bone,Name="thigh_r")
    FalloffRegionScaleFactors=(PositiveWidth=1.5,NegativeWidth=1.25,PositiveHeight=1,NegativeHeight=1)
    FalloffSize=0.45
    OptionalParentItem=(Type=Bone,Name="pelvis")
    RotationOffset=(X=-150,Y=0,Z=10)
   )
   > &RigUnit_ModifyTransforms(
    ItemToModify=((Item=(Type=Bone,Name="jacket_frt_02_r"),Transform=(Rotation=(X=0,Y=0,Z=0,W=1),Translation=(X=0,Y=0,Z=0),Scale3D=(X=1,Y=1,Z=1))))
    ItemToModify.0.Transform.Rotation=&RigVMFunction_MathQuaternionFromEuler(Euler=(X=0,Y=0,Z=0), RotationOrder=XYZ).Result
    Weight=@n8.OutputParam
   )
   > &RigUnit_ModifyTransforms(
    ItemToModify=((Item=(Type=Bone,Name="jacket_frtSide_01_r"),Transform=(Rotation=(X=0,Y=0,Z=0,W=1),Translation=(X=0,Y=0,Z=0),Scale3D=(X=1,Y=1,Z=1))))
    ItemToModify.0.Transform.Rotation=&RigVMFunction_MathQuaternionFromEuler(Euler=(X=0,Y=0,Z=0), RotationOrder=XYZ).Result
    Weight=@n8.OutputParam
   )
   > &RigUnit_ModifyTransforms(
    ItemToModify=((Item=(Type=Bone,Name="jacket_side_01_r"),Transform=(Rotation=(X=0,Y=0,Z=0,W=1),Translation=(X=0,Y=0,Z=0),Scale3D=(X=1,Y=1,Z=1))))
    ItemToModify.0.Transform.Rotation=&RigVMFunction_MathQuaternionFromEuler(Euler=(X=0,Y=0,Z=0), RotationOrder=XYZ).Result
    Weight=@n8.OutputParam
   )
   n9 = &RigUnit_SphericalPoseReader(
    ActiveRegionScaleFactors=(PositiveWidth=0.75,NegativeWidth=0.75,PositiveHeight=0.75,NegativeHeight=0.75)
    ActiveRegionSize=0.1
    Debug=(bDrawDebug=true,bDraw2D=false,bDrawLocalAxes=false,DebugScale=25,DebugSegments=20,DebugThickness=0.25)
    DriverAxis=(X=1,Y=0,Z=0)
    DriverItem=(Type=Bone,Name="spine_05")
    ExecutePin=()
    FalloffRegionScaleFactors=(PositiveWidth=0.5,NegativeWidth=0.5,PositiveHeight=0.5,NegativeHeight=0.5)
    FalloffSize=0.2
    OptionalParentItem=(Type=Bone,Name="pelvis")
    RotationOffset=(X=90,Y=0,Z=-65)
   )
   > &RigUnit_ModifyTransforms(
    ItemToModify=((Item=(Type=Bone,Name="jacket_frt_01_l"),Transform=(Rotation=(X=0,Y=0,Z=0,W=1),Translation=(X=0,Y=0,Z=0),Scale3D=(X=1,Y=1,Z=1))))
    ItemToModify.0.Transform.Rotation=&RigVMFunction_MathQuaternionFromEuler(Euler=(X=0,Y=0,Z=0), RotationOrder=XYZ).Result
    Weight=@n9.OutputParam
   )
   > &RigUnit_ModifyTransforms(
    ItemToModify=((Item=(Type=Bone,Name="jacket_frtSide_01_l"),Transform=(Rotation=(X=0,Y=0,Z=0,W=1),Translation=(X=0,Y=0,Z=0),Scale3D=(X=1,Y=1,Z=1))))
    ItemToModify.0.Transform.Rotation=&RigVMFunction_MathQuaternionFromEuler(Euler=(X=0,Y=0,Z=0), RotationOrder=XYZ).Result
    Weight=@n9.OutputParam
   )
   > &RigUnit_ModifyTransforms(
    ItemToModify=((Item=(Type=Bone,Name="jacket_side_01_l"),Transform=(Rotation=(X=0,Y=0,Z=0,W=1),Translation=(X=0,Y=0,Z=0),Scale3D=(X=1,Y=1,Z=1))))
    ItemToModify.0.Transform.Rotation=&RigVMFunction_MathQuaternionFromEuler(Euler=(X=0,Y=0,Z=0), RotationOrder=XYZ).Result
    Weight=@n9.OutputParam
   )
   > &RigUnit_ModifyTransforms(
    ItemToModify=((Item=(Type=Bone,Name="jacket_frt_01_r"),Transform=(Rotation=(X=0,Y=0,Z=0,W=1),Translation=(X=0,Y=0,Z=0),Scale3D=(X=1,Y=1,Z=1))))
    ItemToModify.0.Transform.Rotation=&RigVMFunction_MathQuaternionFromEuler(Euler=(X=0,Y=0,Z=0), RotationOrder=XYZ).Result
    Weight=@n9.OutputParam
   )
   > &RigUnit_ModifyTransforms(
    ItemToModify=((Item=(Type=Bone,Name="jacket_frtSide_01_r"),Transform=(Rotation=(X=0,Y=0,Z=0,W=1),Translation=(X=0,Y=0,Z=0),Scale3D=(X=1,Y=1,Z=1))))
    ItemToModify.0.Transform.Rotation=&RigVMFunction_MathQuaternionFromEuler(Euler=(X=0,Y=0,Z=0), RotationOrder=XYZ).Result
    Weight=@n9.OutputParam
   )
   > &RigUnit_ModifyTransforms(
    ItemToModify=((Item=(Type=Bone,Name="jacket_side_01_r"),Transform=(Rotation=(X=0,Y=0,Z=0,W=1),Translation=(X=0,Y=0,Z=0),Scale3D=(X=1,Y=1,Z=1))))
    ItemToModify.0.Transform.Rotation=&RigVMFunction_MathQuaternionFromEuler(Euler=(X=0,Y=0,Z=0), RotationOrder=XYZ).Result
    Weight=@n9.OutputParam
   )
  ]}
]
hierarchy [
 root {loc=<0 0 0>, rot=<0 0 0 1>}
 pelvis {parent=root, loc=<0 -0.7071 93.1191>, rot=<0.0127 0.707 -0.0127 -0.707>}
 spine_01 {parent=pelvis, loc=<6.7494 -0.6388 0>, rot=<0 0 0.034 -0.9994>}
 spine_02 {parent=spine_01, loc=<7.2518 -0.035 0>, rot=<0 0 0.0238 -0.9997>}
 spine_03 {parent=spine_02, loc=<6.0104 0 0>, rot=<0 0 0.0149 -0.9999>}
 spine_04 {parent=spine_03, loc=<8.1024 0 0>, rot=<0 0 -0.0132 -0.9999>}
 spine_05 {parent=spine_04, loc=<10.0146 0 0>, rot=<0 0 -0.0015 -1>}
 clavicle_l {parent=spine_05, loc=<4.3575 -0.1466 -2.3223>, rot=<0.0465 -0.7289 -0.009 -0.6829>}
 upperarm_l {parent=clavicle_l, loc=<15.5111 0 0>, rot=<0.0226 -0.3234 -0.0088 -0.9459>}
 lowerarm_l {parent=upperarm_l, loc=<26.6514 0 0>, rot=<0.0391 -0.0099 -0.2453 -0.9686>}
 hand_l {parent=lowerarm_l, loc=<22.3491 0 0>, rot=<-0.7081 -0.0507 0.0975 -0.6975>}
 index_metacarpal_l {parent=hand_l, loc=<1.6535 0.0183 -2.96>, rot=<0.0158 -0.0124 -0.0173 -0.9996>}
 index_01_l {parent=index_metacarpal_l, loc=<8.3008 0.0357 -0.0524>, rot=<-0.0044 0.0485 0.0924 -0.9945>}
 index_02_l {parent=index_01_l, loc=<3.3587 -0.3004 0.0452>, rot=<-0.002 0.0027 0.3225 -0.9466>}
 index_03_l {parent=index_02_l, loc=<2.5578 -0.123 -0.0252>, rot=<-0.002 0.0067 0.2558 -0.9667>}
 middle_metacarpal_l {parent=hand_l, loc=<2.1856 -0.1725 -0.5984>, rot=<0.006 0.0168 -0.0329 -0.9993>}
 middle_01_l {parent=middle_metacarpal_l, loc=<7.3275 -0.2447 0.0129>, rot=<-0.0208 0.0042 0.0858 -0.9961>}
 middle_02_l {parent=middle_01_l, loc=<4.0123 0.1401 0.1175>, rot=<-0.0218 0.0258 0.3512 -0.9357>}
 middle_03_l {parent=middle_02_l, loc=<2.6077 0.0534 0.1373>, rot=<0.0082 0.0188 0.1708 -0.9851>}
 pinky_metacarpal_l {parent=hand_l, loc=<1.7616 -0.9753 3.3814>, rot=<-0.0397 0.0814 -0.069 -0.9935>}
 pinky_01_l {parent=pinky_metacarpal_l, loc=<7.1709 -0.4831 -0.1094>, rot=<-0.0654 -0.0035 0.2242 -0.9723>}
 pinky_02_l {parent=pinky_01_l, loc=<2.8274 0.0893 -0.1043>, rot=<0.0037 0.0011 0.2854 -0.9584>}
 pinky_03_l {parent=pinky_02_l, loc=<2.2046 0.0307 -0.0078>, rot=<-0.0002 0.0098 0.1445 -0.9895>}
 ring_metacarpal_l {parent=hand_l, loc=<2.2074 -0.3625 1.3638>, rot=<-0.0284 0.0722 -0.0458 -0.9959>}
 ring_01_l {parent=ring_metacarpal_l, loc=<7.0295 -0.6197 -0.1446>, rot=<-0.0489 -0.0252 0.1892 -0.9804>}
 ring_02_l {parent=ring_01_l, loc=<3.6174 0.4647 -0.1027>, rot=<0 0 0.2806 -0.9598>}
 ring_03_l {parent=ring_02_l, loc=<2.4152 0.114 0.0199>, rot=<0 0 0.1189 -0.9929>}
 thumb_01_l {parent=hand_l, loc=<1.3118 -0.3191 -2.459>, rot=<0.6546 -0.4042 0.0664 -0.6354>}
 thumb_02_l {parent=thumb_01_l, loc=<5.9336 -0.0448 0.0401>, rot=<-0.0037 0.0213 0.298 -0.9543>}
 thumb_03_l {parent=thumb_02_l, loc=<2.2223 0.3682 -0.0907>, rot=<0.0118 -0.013 0.0066 -0.9998>}
 lowerarm_twist_01_l {parent=lowerarm_l, loc=<17.8549 0 0>, rot=<0 0 0 1>}
 lowerarm_twist_02_l {parent=lowerarm_l, loc=<9.6199 0 0>, rot=<0 0 0 1>}
 upperarm_twist_01_l {parent=upperarm_l, loc=<0.0041 0 0>, rot=<0 0 0 1>}
 upperarm_twist_02_l {parent=upperarm_l, loc=<13.7941 0 0>, rot=<0 0 0 1>}
 dyn_parent_shldr_tipA_l {parent=clavicle_l, loc=<12.1525 8.7438 7.4266>, rot=<-0.0915 0.5379 0.1993 -0.814>}
 dyn_shldr_tipA_01_l {parent=dyn_parent_shldr_tipA_l, loc=<0 0 0>, rot=<0 0 0 1>}
 dyn_shldr_tipA_02_l {parent=dyn_shldr_tipA_01_l, loc=<3.984 0 0>, rot=<0 0 0 1>}
 dyn_parent_shldr_tipB_l {parent=clavicle_l, loc=<15.3527 7.5806 6.9171>, rot=<-0.0956 0.5209 0.1973 -0.825>}
 dyn_shldr_tipB_01_l {parent=dyn_parent_shldr_tipB_l, loc=<0 0 0>, rot=<0 0 0 1>}
 dyn_shldr_tipB_02_l {parent=dyn_shldr_tipB_01_l, loc=<3.9263 0 0>, rot=<0 0 0 1>}
 dyn_parent_cowl_l {parent=clavicle_l, loc=<10.5945 14.9418 -4.9802>, rot=<0.0008 0.0336 -0.0252 -0.9991>}
 dyn_cowl_01_l {parent=dyn_parent_cowl_l, loc=<0 0 0>, rot=<-0.6609 0.1705 0.7008 0.2074>}
 dyn_cowl_02_l {parent=dyn_cowl_01_l, loc=<6.8749 0 0>, rot=<0 0 0 1>}
 dyn_cowl_03_l {parent=dyn_cowl_02_l, loc=<6.677 0 0>, rot=<0 0 0 1>}
 clavicle_r {parent=spine_05, loc=<4.3575 -0.1466 2.3223>, rot=<0.7289 0.0465 -0.6829 0.009>}
 upperarm_r {parent=clavicle_r, loc=<-15.5111 0 0>, rot=<0.0226 -0.3234 -0.0088 -0.9459>}
 lowerarm_r {parent=upperarm_r, loc=<-26.6515 0 0>, rot=<0.0391 -0.0099 -0.2453 -0.9686>}
 hand_r {parent=lowerarm_r, loc=<-22.3492 0 0>, rot=<-0.7081 -0.0507 0.0975 -0.6975>}
 index_metacarpal_r {parent=hand_r, loc=<-1.6535 -0.0183 2.96>, rot=<0.0158 -0.0124 -0.0173 -0.9996>}
 index_01_r {parent=index_metacarpal_r, loc=<-8.3008 -0.0357 0.0524>, rot=<-0.0044 0.0485 0.0924 -0.9945>}
 index_02_r {parent=index_01_r, loc=<-3.3587 0.3004 -0.0452>, rot=<-0.002 0.0027 0.3225 -0.9466>}
 index_03_r {parent=index_02_r, loc=<-2.5578 0.1231 0.0251>, rot=<-0.002 0.0067 0.2558 -0.9667>}
 middle_metacarpal_r {parent=hand_r, loc=<-2.1856 0.1725 0.5984>, rot=<0.006 0.0168 -0.0329 -0.9993>}
 middle_01_r {parent=middle_metacarpal_r, loc=<-7.3275 0.2447 -0.0129>, rot=<-0.0208 0.0042 0.0858 -0.9961>}
 middle_02_r {parent=middle_01_r, loc=<-4.0123 -0.1401 -0.1175>, rot=<-0.0218 0.0258 0.3512 -0.9357>}
 middle_03_r {parent=middle_02_r, loc=<-2.6077 -0.0534 -0.1373>, rot=<0.0082 0.0188 0.1708 -0.9851>}
 pinky_metacarpal_r {parent=hand_r, loc=<-1.7616 0.9753 -3.3814>, rot=<-0.0397 0.0814 -0.069 -0.9935>}
 pinky_01_r {parent=pinky_metacarpal_r, loc=<-7.1709 0.4831 0.1094>, rot=<-0.0654 -0.0035 0.2242 -0.9723>}
 pinky_02_r {parent=pinky_01_r, loc=<-2.8274 -0.0892 0.1043>, rot=<0.0037 0.0011 0.2854 -0.9584>}
 pinky_03_r {parent=pinky_02_r, loc=<-2.2046 -0.0307 0.0078>, rot=<-0.0002 0.0098 0.1445 -0.9895>}
 ring_metacarpal_r {parent=hand_r, loc=<-2.2074 0.3625 -1.3638>, rot=<-0.0284 0.0722 -0.0458 -0.9959>}
 ring_01_r {parent=ring_metacarpal_r, loc=<-7.0295 0.6197 0.1446>, rot=<-0.0489 -0.0252 0.1892 -0.9804>}
 ring_02_r {parent=ring_01_r, loc=<-3.6174 -0.4647 0.1027>, rot=<0 0 0.2806 -0.9598>}
 ring_03_r {parent=ring_02_r, loc=<-2.4153 -0.1141 -0.0199>, rot=<0 0 0.1189 -0.9929>}
 thumb_01_r {parent=hand_r, loc=<-1.3118 0.3191 2.459>, rot=<0.6546 -0.4042 0.0664 -0.6354>}
 thumb_02_r {parent=thumb_01_r, loc=<-5.9335 0.0447 -0.0401>, rot=<-0.0037 0.0213 0.298 -0.9543>}
 thumb_03_r {parent=thumb_02_r, loc=<-2.2223 -0.3682 0.0907>, rot=<0.0118 -0.013 0.0066 -0.9998>}
 lowerarm_twist_01_r {parent=lowerarm_r, loc=<-17.855 0 0>, rot=<0 0 0 1>}
 lowerarm_twist_02_r {parent=lowerarm_r, loc=<-9.6199 0 0>, rot=<0 0 0 1>}
 upperarm_twist_01_r {parent=upperarm_r, loc=<-0.0041 0 0>, rot=<0 0 0 1>}
 upperarm_twist_02_r {parent=upperarm_r, loc=<-13.7942 0 0>, rot=<0 0 0 1>}
 dyn_parent_cowl_r {parent=clavicle_r, loc=<-10.5949 -14.9419 4.9803>, rot=<0.9991 0.0252 0.0336 0.0008>}
 dyn_cowl_01_r {parent=dyn_parent_cowl_r, loc=<0.0005 0 0>, rot=<0.707 0.181 -0.6687 0.1426>}
 dyn_cowl_02_r {parent=dyn_cowl_01_r, loc=<6.8752 0 0>, rot=<0 0 0 -1>}
 dyn_cowl_03_r {parent=dyn_cowl_02_r, loc=<6.6766 0 0>, rot=<0 0 0 1>}
 neck_01 {parent=spine_05, loc=<11.3138 0.1529 0>, rot=<0 0 -0.2003 -0.9797>}
 neck_02 {parent=neck_01, loc=<4.5627 0 0>, rot=<0 0 0.0504 -0.9987>}
 head {parent=neck_02, loc=<5.3914 0 0>, rot=<0 0 0.115 -0.9934>}
 neck_offset {parent=neck_01, loc=<0 0 0>, rot=<0 0 0 -1>}
 attach_cape {parent=spine_05, loc=<9.017 -11.714 0>, rot=<0.0616 -0.7044 -0.0616 -0.7044>}
 attach_backpack {parent=spine_05, loc=<-0.123 -13.473 0>, rot=<-0.0093 -0.707 0.0093 -0.707>}
 jacket_base {parent=spine_04, loc=<3.0159 0 0>, rot=<0 0 0 1>}
 jacket_frt_01_l {parent=jacket_base, loc=<2.0217 15.383 -6.1862>, rot=<0.7592 0.0407 -0.023 -0.6491>}
 jacket_frt_02_l {parent=jacket_frt_01_l, loc=<-7.4836 0 0>, rot=<-0.0003 0.002 0.0026 -1>}
 jacket_frt_03_l {parent=jacket_frt_02_l, loc=<-9.1117 -0.052 0.0208>, rot=<-0.0022 0.0115 0.0277 -0.9995>}
 jacket_frt_04_l {parent=jacket_frt_03_l, loc=<-10.5887 -0.0258 -0.1699>, rot=<-0.0008 0.0059 -0.0049 -1>}
 jacket_frt_05_l {parent=jacket_frt_04_l, loc=<-13.4773 -0.2185 -0.1147>, rot=<0.0023 -0.0198 0.0197 -0.9996>}
 jacket_frt_06_l {parent=jacket_frt_05_l, loc=<-16.1781 0.1587 0.228>, rot=<0.0025 -0.0196 -0.0034 -0.9998>}
 jacket_frt_07_l {parent=jacket_frt_06_l, loc=<-12.9333 0.0052 -0.0328>, rot=<0.0004 -0.0038 -0.0004 -1>}
 jacket_frt_08_l {parent=jacket_frt_07_l, loc=<-10.6367 0 0>, rot=<0 0 0 1>}
 jacket_frt_09_l {parent=jacket_frt_08_l, loc=<-8.7495 0 0>, rot=<0 0 0 1>}
 jacket_frt_01_r {parent=jacket_base, loc=<2.022 15.3831 6.3048>, rot=<-0.7592 -0.0407 -0.023 -0.6491>}
 jacket_frt_02_r {parent=jacket_frt_01_r, loc=<-7.484 0 0>, rot=<0.0003 -0.002 0.0026 -1>}
 jacket_frt_03_r {parent=jacket_frt_02_r, loc=<-9.1114 -0.052 -0.0208>, rot=<0.0022 -0.0115 0.0277 -0.9995>}
 jacket_frt_04_r {parent=jacket_frt_03_r, loc=<-10.5883 -0.0257 0.1699>, rot=<0.0008 -0.0059 -0.0049 -1>}
 jacket_frt_05_r {parent=jacket_frt_04_r, loc=<-13.4778 -0.2185 0.1147>, rot=<-0.0023 0.0198 0.0197 -0.9996>}
 jacket_frt_06_r {parent=jacket_frt_05_r, loc=<-16.1781 0.1587 -0.228>, rot=<-0.0025 0.0196 -0.0034 -0.9998>}
 jacket_frt_07_r {parent=jacket_frt_06_r, loc=<-12.9332 0.0051 0.0328>, rot=<-0.0004 0.0038 -0.0004 -1>}
 jacket_frt_08_r {parent=jacket_frt_07_r, loc=<-10.6367 0 0>, rot=<0 0 0 1>}
 jacket_frt_09_r {parent=jacket_frt_08_r, loc=<-8.7496 0 0>, rot=<0 0 0 1>}
 jacket_frtSide_01_l {parent=jacket_base, loc=<-4.3331 8.5299 -11.6612>, rot=<0.7634 0.0625 0.021 -0.6425>}
 jacket_frtSide_02_l {parent=jacket_frtSide_01_l, loc=<-7.3148 0.0164 -0.0041>, rot=<-0.0081 0.0191 0.0224 -0.9995>}
 jacket_frtSide_03_l {parent=jacket_frtSide_02_l, loc=<-12.844 -0.0039 0.0004>, rot=<0.0004 -0.0007 -0.008 -1>}
 jacket_frtSide_04_l {parent=jacket_frtSide_03_l, loc=<-13.0754 0.0186 0.0002>, rot=<0.0002 0.0015 -0.0038 -1>}
 jacket_frtSide_05_l {parent=jacket_frtSide_04_l, loc=<-15.2356 -0.0199 0.0127>, rot=<0.0019 -0.0118 -0.0185 -0.9998>}
 jacket_frtSide_06_l {parent=jacket_frtSide_05_l, loc=<-14.5738 0.1078 -0.6422>, rot=<0 0 0 1>}
 jacket_frtSide_07_l {parent=jacket_frtSide_06_l, loc=<-12.1562 0.0884 -0.5262>, rot=<0 0 0 1>}
 jacket_frtSide_08_l {parent=jacket_frtSide_07_l, loc=<-13.9315 0.1363 -0.8114>, rot=<0 0 0 1>}
 jacket_frtSide_01_r {parent=jacket_base, loc=<-4.3334 8.5299 11.6612>, rot=<-0.7634 -0.0625 0.021 -0.6425>}
 jacket_frtSide_02_r {parent=jacket_frtSide_01_r, loc=<-7.3142 0.0165 0.0041>, rot=<0.0086 -0.0211 0.0102 -0.9997>}
 jacket_frtSide_03_r {parent=jacket_frtSide_02_r, loc=<-12.8403 0.3107 -0.0496>, rot=<-0.0016 0.0037 0.0115 -0.9999>}
 jacket_frtSide_04_r {parent=jacket_frtSide_03_r, loc=<-13.0756 0.0185 -0.0002>, rot=<0.0002 -0.0026 -0.0105 -0.9999>}
 jacket_frtSide_05_r {parent=jacket_frtSide_04_r, loc=<-15.2356 -0.0199 -0.0127>, rot=<-0.0019 0.0118 -0.0185 -0.9998>}
 jacket_frtSide_06_r {parent=jacket_frtSide_05_r, loc=<-14.5738 0.1079 0.6422>, rot=<0 0 0 1>}
 jacket_frtSide_07_r {parent=jacket_frtSide_06_r, loc=<-12.1562 0.0884 0.5262>, rot=<0 0 0 1>}
 jacket_frtSide_08_r {parent=jacket_frtSide_07_r, loc=<-13.9315 0.1363 0.8114>, rot=<0 0 0 1>}
 jacket_side_01_l {parent=jacket_base, loc=<-4.0204 1.1839 -11.6616>, rot=<0.7898 0.0694 0.0147 -0.6092>}
 jacket_side_02_l {parent=jacket_side_01_l, loc=<-7.1121 0 0>, rot=<-0.0062 0.0369 0.035 -0.9987>}
 jacket_side_03_l {parent=jacket_side_02_l, loc=<-12.5763 0 0>, rot=<0.0004 0.0008 -0.0047 -1>}
 jacket_side_04_l {parent=jacket_side_03_l, loc=<-12.7758 0 0>, rot=<-0.0003 0.0019 0.0078 -1>}
 jacket_side_05_l {parent=jacket_side_04_l, loc=<-13.1963 0.0586 -0.0305>, rot=<0.0021 -0.0101 0.0008 -0.9999>}
 jacket_side_06_l {parent=jacket_side_05_l, loc=<-13.0422 0.0013 0.0041>, rot=<0.0001 -0.0015 0.0005 -1>}
 jacket_side_07_l {parent=jacket_side_06_l, loc=<-9.2478 0 0>, rot=<0 0 0 1>}
 jacket_side_08_l {parent=jacket_side_07_l, loc=<-8.1342 0 0>, rot=<0.1131 0 0 -0.9936>}
 jacket_side_09_l {parent=jacket_side_08_l, loc=<-10.9315 0 0>, rot=<0 0 0 1>}
 jacket_side_01_r {parent=jacket_base, loc=<-4.0204 1.1839 11.6616>, rot=<-0.7898 -0.0694 0.0147 -0.6092>}
 jacket_side_02_r {parent=jacket_side_01_r, loc=<-7.1116 0.0001 0>, rot=<0.0062 -0.0369 0.035 -0.9987>}
 jacket_side_03_r {parent=jacket_side_02_r, loc=<-12.5772 -0.0002 0>, rot=<-0.0004 -0.0008 -0.0047 -1>}
 jacket_side_04_r {parent=jacket_side_03_r, loc=<-12.7762 -0.0001 0>, rot=<0.0003 -0.0019 0.0078 -1>}
 jacket_side_05_r {parent=jacket_side_04_r, loc=<-13.1963 0.0586 0.0305>, rot=<-0.0021 0.0101 0.0008 -0.9999>}
 jacket_side_06_r {parent=jacket_side_05_r, loc=<-13.0422 0.0012 -0.004>, rot=<-0.0001 0.0015 0.0005 -1>}
 jacket_side_07_r {parent=jacket_side_06_r, loc=<-9.2478 0 0>, rot=<0 0 0 1>}
 jacket_side_08_r {parent=jacket_side_07_r, loc=<-8.1341 0.0001 0>, rot=<-0.1131 0 0 -0.9936>}
 jacket_side_09_r {parent=jacket_side_08_r, loc=<-10.9315 0 0>, rot=<0 0 0 1>}
 jacket_back_01_l {parent=jacket_base, loc=<-3.3743 -6.8676 -6.4049>, rot=<0.9862 0.0857 0.0258 -0.1393>}
 jacket_back_02_l {parent=jacket_back_01_l, loc=<-7.7994 0 0>, rot=<0 0 0 1>}
 jacket_back_03_l {parent=jacket_back_02_l, loc=<-11.9548 -0.039 0.0043>, rot=<0.0037 -0.0046 0.0425 -0.9991>}
 jacket_back_04_l {parent=jacket_back_03_l, loc=<-12.3757 0.0586 -0.0305>, rot=<0 0 0 1>}
 jacket_back_05_l {parent=jacket_back_04_l, loc=<-11.5256 -0.0008 -0.0027>, rot=<0 0 0 1>}
 jacket_back_06_l {parent=jacket_back_05_l, loc=<-9.499 0 0>, rot=<0 0 0 1>}
 jacket_back_07_l {parent=jacket_back_06_l, loc=<-9.5747 0 0>, rot=<0 0 0 1>}
 jacket_back_08_l {parent=jacket_back_07_l, loc=<-9.2544 0 0>, rot=<0 0 0 1>}
 jacket_back_09_l {parent=jacket_back_08_l, loc=<-13.2396 0 0>, rot=<0 0 0 1>}
 jacket_back_01_r {parent=jacket_base, loc=<-3.3156 -6.936 6.7497>, rot=<0.9907 0.0819 -0.0331 0.1039>}
 jacket_back_02_r {parent=jacket_back_01_r, loc=<-7.7994 0 0>, rot=<0 0 0 1>}
 jacket_back_03_r {parent=jacket_back_02_r, loc=<-11.9548 -0.039 0.0043>, rot=<0.0037 -0.0046 0.0425 -0.9991>}
 jacket_back_04_r {parent=jacket_back_03_r, loc=<-12.3757 0.0586 -0.0305>, rot=<0 0 0 1>}
 jacket_back_05_r {parent=jacket_back_04_r, loc=<-11.5256 -0.0008 -0.0027>, rot=<0 0 0 1>}
 jacket_back_06_r {parent=jacket_back_05_r, loc=<-9.499 0 0>, rot=<0 0 0 1>}
 jacket_back_07_r {parent=jacket_back_06_r, loc=<-9.5747 0 0>, rot=<0 0 0 1>}
 jacket_back_08_r {parent=jacket_back_07_r, loc=<-9.2544 0 0>, rot=<0 0 0 1>}
 jacket_back_09_r {parent=jacket_back_08_r, loc=<-13.2396 0 0>, rot=<0 0 0 1>}
 jacket_back_01 {parent=jacket_base, loc=<-3.4737 -8.3026 -0.0312>, rot=<0.996 0.0886 -0.0001 0.0125>}
 jacket_back_02 {parent=jacket_back_01, loc=<-7.7994 0 0>, rot=<0 0 0.022 -0.9998>}
 jacket_back_03 {parent=jacket_back_02, loc=<-11.9548 -0.039 0.0043>, rot=<0.0036 -0.0047 0.0206 -0.9998>}
 jacket_back_04 {parent=jacket_back_03, loc=<-12.3757 0.0586 -0.0305>, rot=<0 0 0.0015 -1>}
 jacket_back_05 {parent=jacket_back_04, loc=<-11.5256 -0.0008 -0.0027>, rot=<0 0 0 1>}
 jacket_back_06 {parent=jacket_back_05, loc=<-8.6478 0 0>, rot=<0 0 0 1>}
 jacket_back_07 {parent=jacket_back_06, loc=<-9.3468 0 0>, rot=<0 0 0 1>}
 jacket_back_08 {parent=jacket_back_07, loc=<-9.2883 0 0>, rot=<0 0 0 1>}
 jacket_back_09 {parent=jacket_back_08, loc=<-13.257 0 0>, rot=<0 0 0 1>}
 thigh_l {parent=pelvis, loc=<-2.5693 -0.0109 -7.7501>, rot=<0.0012 0.0572 0.0155 -0.9982>}
 calf_l {parent=thigh_l, loc=<-39.0484 0 0>, rot=<0.0006 -0.0161 -0.0391 -0.9991>}
 calf_twist_02_l {parent=calf_l, loc=<-18.8064 -0.0967 0.0397>, rot=<0.0237 0.0095 -0.0035 -0.9997>}
 foot_l {parent=calf_l, loc=<-42.561 0.1077 0.4292>, rot=<0.0418 -0.0455 0.0343 -0.9975>}
 ball_l {parent=foot_l, loc=<-6.1599 13.8264 0.6806>, rot=<0.0001 -0.0001 -0.7063 -0.7079>}
 calf_twist_01_l {parent=calf_l, loc=<-35.813 -0.0967 0.0397>, rot=<0.0237 0.0095 -0.0035 -0.9997>}
 thigh_twist_01_l {parent=thigh_l, loc=<-21.6303 0 0>, rot=<0.0013 0 0 -1>}
 thigh_r {parent=pelvis, loc=<-2.5693 -0.0109 7.7501>, rot=<0.0572 -0.0012 0.9982 0.0155>}
 calf_r {parent=thigh_r, loc=<39.0484 0 0>, rot=<0.0006 -0.0161 -0.0391 -0.9991>}
 calf_twist_02_r {parent=calf_r, loc=<18.8064 0.0967 -0.0397>, rot=<0.0237 0.0095 -0.0035 -0.9997>}
 foot_r {parent=calf_r, loc=<42.561 -0.1077 -0.4292>, rot=<0.0418 -0.0455 0.0343 -0.9975>}
 ball_r {parent=foot_r, loc=<6.1598 -13.8264 -0.6806>, rot=<0.0001 -0.0001 -0.7063 -0.7079>}
 calf_twist_01_r {parent=calf_r, loc=<35.813 0.0967 -0.0397>, rot=<0.0237 0.0095 -0.0035 -0.9997>}
 thigh_twist_01_r {parent=thigh_r, loc=<21.6303 0 0>, rot=<0.0013 0 0 -1>}
 jacket_frt_03_l {type=Curve, loc=<0 0 0>, rot=<0 0 0 1>}
 jacket_frt_03_r {type=Curve, loc=<0 0 0>, rot=<0 0 0 1>}
 jacket_frt_04_l {type=Curve, loc=<0 0 0>, rot=<0 0 0 1>}
 spine_03 {type=Curve, loc=<0 0 0>, rot=<0 0 0 1>}
 jacket_frt_02_r {type=Curve, loc=<0 0 0>, rot=<0 0 0 1>}
 jacket_frt_02_l {type=Curve, loc=<0 0 0>, rot=<0 0 0 1>}
 jacket_frtSide_02_l {type=Curve, loc=<0 0 0>, rot=<0 0 0 1>}
 jacket_frtSide_02_r {type=Curve, loc=<0 0 0>, rot=<0 0 0 1>}
 jacket_side_02_r {type=Curve, loc=<0 0 0>, rot=<0 0 0 1>}
 jacket_side_02_l {type=Curve, loc=<0 0 0>, rot=<0 0 0 1>}
 jacket_back_02_l {type=Curve, loc=<0 0 0>, rot=<0 0 0 1>}
 jacket_back_02_r {type=Curve, loc=<0 0 0>, rot=<0 0 0 1>}
 jacket_back_02 {type=Curve, loc=<0 0 0>, rot=<0 0 0 1>}
 jacket_frtSide_03_l {type=Curve, loc=<0 0 0>, rot=<0 0 0 1>}
 jacket_frtSide_03_r {type=Curve, loc=<0 0 0>, rot=<0 0 0 1>}
]
variables [
 Is Crouching Float {type=double, default=0}
 Is Pistol Alpha {type=double, default=0}
 Is Rifle Alpha {type=double, default=0}
]
