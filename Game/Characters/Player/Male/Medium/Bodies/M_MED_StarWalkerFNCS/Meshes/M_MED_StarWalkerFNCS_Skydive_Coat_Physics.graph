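# Physics asset for the StarWalkerFNCS male medium body's coat, named for skydiving, with capsule bodies on coat bones, pelvis, spine_03 and legs.
# Limited-angle constraint chains for back, side, front and front-side panels branch from jacket_base, which joins spine_03 with free rotation.
# Cross constraints tie the center back chain to both neighbors and left front-side to left front.
# The right front lacks these ties and has partly looser limits.
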PhysicsAsset /Game/Characters/Player/Male/Medium/Bodies/M_MED_StarWalkerFNCS/Meshes/M_MED_StarWalkerFNCS_Skydive_Coat_Physics
bodies {
 calf_l [{shapeType=Capsule, length=30, radius=8, center=<-21.253 -2.7132 0.0004>, rotation=<rot 90 0 -178.9129>}]
 calf_r [{shapeType=Capsule, length=30, radius=8, center=<21.253 2.7132 0.0004>, rotation=<rot 90 0 -178.9129>}]
 jacket_back_01 [{shapeType=Capsule, length=8, radius=10, center=<-3 8 0>, rotation=<rot 90 0 0>}]
 jacket_back_01_l [{shapeType=Capsule, length=8, radius=10, center=<-3 8 0>, rotation=<rot 90 15.4568 15.4568>}]
 jacket_back_01_r [{shapeType=Capsule, length=8, radius=10, center=<-3 8 0>, rotation=<rot 90 0 0>}]
 jacket_back_02 [{shapeType=Capsule, length=8, radius=10, center=<-3 8 0>, rotation=<rot 90 0 0>}]
 jacket_back_02_l [{shapeType=Capsule, length=8, radius=10, center=<-3 8 0>, rotation=<rot 90 15.4568 15.4568>}]
 jacket_back_02_r [{shapeType=Capsule, length=8, radius=10, center=<-3 8 0>, rotation=<rot 90 -157.5834 -157.5834>}]
 jacket_back_03 [{shapeType=Capsule, length=8, radius=10, center=<-3 8 0>, rotation=<rot 90 0 0>}]
 jacket_back_03_l [{shapeType=Capsule, length=8, radius=10, center=<-3 8 0>, rotation=<rot 90 22.5263 22.5263>}]
 jacket_back_03_r [{shapeType=Capsule, length=8, radius=10, center=<-3 8 0>, rotation=<rot 90 -155.6401 -155.6401>}]
 jacket_back_04 [{shapeType=Capsule, length=8, radius=10, center=<-3 8 0>, rotation=<rot 90 0 0>}]
 jacket_back_04_l [{shapeType=Capsule, length=8, radius=10, center=<-3 8 0>, rotation=<rot 90 22.5263 22.5263>}]
 jacket_back_04_r [{shapeType=Capsule, length=8, radius=10, center=<-3 8 0>, rotation=<rot 90 -155.6401 -155.6401>}]
 jacket_back_05 [{shapeType=Capsule, length=8, radius=10, center=<-3 8 0>, rotation=<rot 90 0 0>}]
 jacket_back_05_l [{shapeType=Capsule, length=8, radius=10, center=<-3 8 0>, rotation=<rot 90 22.5263 22.5263>}]
 jacket_back_05_r [{shapeType=Capsule, length=8, radius=10, center=<-3 8 0>, rotation=<rot 90 -155.6401 -155.6401>}]
 jacket_back_06_l [{shapeType=Capsule, length=8, radius=10, center=<-3 8 0>, rotation=<rot 90 22.5263 22.5263>}]
 jacket_back_06_r [{shapeType=Capsule, length=8, radius=10, center=<-3 8 0>, rotation=<rot 90 -155.6401 -155.6401>}]
 jacket_back_07_l [{shapeType=Capsule, length=8, radius=10, center=<-3 8 0>, rotation=<rot 90 22.5263 22.5263>}]
 jacket_back_07_r [{shapeType=Capsule, length=8, radius=10, center=<-3 8 0>, rotation=<rot 90 -155.6401 -155.6401>}]
 jacket_base [{shapeType=Capsule, length=0.505, radius=0.505, rotation=<rot 0 -90 0>}]
 jacket_frtSide_01_l [{shapeType=Capsule, length=10.25, radius=10, center=<-3 10 0>, rotation=<rot 0 0 -20>}]
 jacket_frtSide_01_r [{shapeType=Capsule, length=10.25, radius=10, center=<-3 10 0>, rotation=<rot 0 0 20>}]
 jacket_frtSide_02_l [{shapeType=Capsule, length=8, radius=10, center=<-3 8 0>, rotation=<rot 0 0 -20>}]
 jacket_frtSide_02_r [{shapeType=Capsule, length=8, radius=10, center=<-3 6 0>, rotation=<rot 0 0 20>}]
 jacket_frtSide_03_l [{shapeType=Capsule, length=8, radius=4, center=<-3 2 0>, rotation=<rot 0 0 -20>}]
 jacket_frtSide_03_r [{shapeType=Capsule, length=8, radius=4, center=<-3 2 0>, rotation=<rot 0 0 20>}]
 jacket_frtSide_04_l [{shapeType=Capsule, length=8, radius=4, center=<-3 2 0>, rotation=<rot 0 0 -20>}]
 jacket_frtSide_04_r [{shapeType=Capsule, length=8, radius=4, center=<-3 2 0>, rotation=<rot 0 0 20>}]
 jacket_frtSide_05_l [{shapeType=Capsule, length=8, radius=4, center=<-3 2 0>, rotation=<rot 0 0 -20>}]
 jacket_frtSide_05_r [{shapeType=Capsule, length=8, radius=4, center=<-3 2 0>, rotation=<rot 0 0 20>}]
 jacket_frt_01_l [{shapeType=Capsule, length=5.75, radius=4.5, center=<-1.9999 0 4.5>, rotation=<rot 90 0 0>}]
 jacket_frt_01_r [{shapeType=Capsule, length=5.75, radius=4.5, center=<-2 0 -4.5>, rotation=<rot 90 0 180>}]
 jacket_frt_02_l [{shapeType=Capsule, length=8, radius=8, center=<-3 8 0>, rotation=<rot 0 0 130.31>}]
 jacket_frt_02_r [{shapeType=Capsule, length=8, radius=8, center=<-3 8 0>, rotation=<rot 0 0 -130>}]
 jacket_frt_03_l [{shapeType=Capsule, length=8, radius=4, center=<-3 4 0>, rotation=<rot 90 0 0>}]
 jacket_frt_03_r [{shapeType=Capsule, length=8, radius=4, center=<-3 4 0>, rotation=<rot 90 0 0>}]
 jacket_frt_04_l [{shapeType=Capsule, length=8, radius=4, center=<-8 4 0>, rotation=<rot 90 0 0>}]
 jacket_frt_04_r [{shapeType=Capsule, length=8, radius=4, center=<-8 4 0>, rotation=<rot 90 0 0>}]
 jacket_frt_05_l [{shapeType=Capsule, length=8, radius=4, center=<-8 6 -2>, rotation=<rot 0 0 -60>}]
 jacket_frt_05_r [{shapeType=Capsule, length=8, radius=4, center=<-8 6 2>, rotation=<rot 0 0 60>}]
 jacket_frt_06_l [{shapeType=Capsule, length=6, radius=4, center=<-3 6 -2>, rotation=<rot 0 0 -60>}]
 jacket_frt_06_r [{shapeType=Capsule, length=6, radius=4, center=<-3 6 2>, rotation=<rot 0 0 60>}]
 jacket_frt_07_l [{shapeType=Capsule, length=6, radius=6, center=<-3 6 1>, rotation=<rot 90 0 0>}]
 jacket_frt_07_r [{shapeType=Capsule, length=6, radius=6, center=<-3 6 -1>, rotation=<rot 90 0 0>}]
 jacket_frt_08_l [{shapeType=Capsule, length=6, radius=6, shapeName=jacket_frt_08_l_capsule, center=<-6 6 2>, rotation=<rot 90 0 0>}]
 jacket_frt_08_r [{shapeType=Capsule, length=6, radius=6, shapeName=jacket_frt_08_r_capsule, center=<-6 6 -2>, rotation=<rot 90 0 0>}]
 jacket_frt_09_l [{shapeType=Capsule, length=6, radius=6, shapeName=jacket_frt_09_l_capsule, center=<-6 6 2>, rotation=<rot 90 0 0>}]
 jacket_frt_09_r [{shapeType=Capsule, length=6, radius=6, shapeName=jacket_frt_09_r_capsule, center=<-6 6 -2>, rotation=<rot 90 0 0>}]
 jacket_side_01_l [{shapeType=Capsule, length=8, radius=10, center=<-6 10 0>, rotation=<rot 0 0 -10>}]
 jacket_side_01_r [{shapeType=Capsule, length=8, radius=10, center=<-6 10 0>, rotation=<rot 0 0 10>}]
 jacket_side_02_l [{shapeType=Capsule, length=8, radius=10, shapeName=jacket_side_02_l_capsule, center=<-6 8 0>, rotation=<rot 0 0 -10>}]
 jacket_side_02_r [{shapeType=Capsule, length=8, radius=10, shapeName=jacket_side_02_r_capsule, center=<-6 8 0>, rotation=<rot 0 0 10>}]
 jacket_side_03_l [{shapeType=Capsule, length=8, radius=6, shapeName=jacket_side_03_l_capsule, center=<-6 2 0>, rotation=<rot 0 0 -10>}]
 jacket_side_03_r [{shapeType=Capsule, length=8, radius=6, shapeName=jacket_side_03_r_capsule, center=<-6 2 0>, rotation=<rot 0 0 10>}]
 jacket_side_04_l [{shapeType=Capsule, length=8, radius=6, shapeName=jacket_side_04_l_capsule, center=<-6 2 0>, rotation=<rot 0 0 -10>}]
 jacket_side_04_r [{shapeType=Capsule, length=8, radius=6, shapeName=jacket_side_04_r_capsule, center=<-6 2 0>, rotation=<rot 0 0 10>}]
 jacket_side_05_l [{shapeType=Capsule, length=8, radius=6, shapeName=jacket_side_05_l_capsule, center=<-6 2 0>, rotation=<rot 0 0 -10>}]
 jacket_side_05_r [{shapeType=Capsule, length=8, radius=6, shapeName=jacket_side_05_r_capsule, center=<-6 2 0>, rotation=<rot 0 0 10>}]
 jacket_side_06_l [{shapeType=Capsule, length=8, radius=6, shapeName=jacket_side_06_l_capsule, center=<-1 2 1>, rotation=<rot 0 0 -10>}]
 jacket_side_06_r [{shapeType=Capsule, length=8, radius=6, shapeName=jacket_side_06_r_capsule, center=<-1 2 -1>, rotation=<rot 0 0 10>}]
 jacket_side_07_l [{shapeType=Capsule, length=8, radius=6, shapeName=jacket_side_07_l_capsule, center=<-12 2 1>, rotation=<rot 0 0 -10>}]
 jacket_side_07_r [{shapeType=Capsule, length=8, radius=6, shapeName=jacket_side_07_r_capsule, center=<-12 2 -1>, rotation=<rot 0 0 10>}]
 jacket_side_08_l [{shapeType=Capsule, length=8, radius=6, shapeName=jacket_side_08_l_capsule, center=<-12 2 1>, rotation=<rot 0 0 -20>}]
 jacket_side_08_r [{shapeType=Capsule, length=8, radius=6, shapeName=jacket_side_08_r_capsule, center=<-12 2 -1>, rotation=<rot 0 0 20>}]
 pelvis [{shapeType=Capsule, length=8.25, radius=11.25, center=<-4.3448 0.7003 -0.1012>}]
 spine_03 [{shapeType=Capsule, length=6.6237, radius=12.6435, center=<-4.485 2.8623 -0.2689>, rotation=<rot 6.2674 -80.8926 90.3064>}]
 thigh_l [{shapeType=Capsule, length=25, radius=7.75, shapeName=thigh_l_capsule, center=<-23 1 0>, rotation=<rot 0 95 90>}]
 thigh_r [{shapeType=Capsule, length=25, radius=7.75, shapeName=thigh_r_capsule, center=<23 -1 0>, rotation=<rot 0 95 90>}]
}
constraints [
 {bone1Name=spine_03, bone2Name=pelvis, swing1LimitDegrees=45, swing2LimitDegrees=45, twistLimitDegrees=45}
 {bone1Name=thigh_r, bone2Name=pelvis, swing1LimitDegrees=45, swing2LimitDegrees=45, twistLimitDegrees=45}
 {bone1Name=calf_r, bone2Name=thigh_r, swing1LimitDegrees=45, swing2LimitDegrees=45, twistLimitDegrees=45}
 {
  bone1Name=jacket_base
  bone2Name=spine_03
  swing1LimitDegrees=45
  swing1Motion=Free
  swing2LimitDegrees=45
  swing2Motion=Free
  twistLimitDegrees=45
  twistMotion=Free
 }
 {bone1Name=jacket_back_01_r, bone2Name=jacket_base, swing1LimitDegrees=22, swing2LimitDegrees=5, twistLimitDegrees=5}
 {bone1Name=jacket_back_02_r, bone2Name=jacket_back_01_r, swing1LimitDegrees=22, swing2LimitDegrees=5, twistLimitDegrees=5}
 {bone1Name=jacket_back_03_r, bone2Name=jacket_back_02_r, swing1LimitDegrees=22, swing2LimitDegrees=5, twistLimitDegrees=5}
 {bone1Name=jacket_back_04_r, bone2Name=jacket_back_03_r, swing1LimitDegrees=22, swing2LimitDegrees=5, twistLimitDegrees=5}
 {bone1Name=jacket_back_05_r, bone2Name=jacket_back_04_r, swing1LimitDegrees=22, swing2LimitDegrees=5, twistLimitDegrees=5}
 {bone1Name=jacket_back_01_l, bone2Name=jacket_base, swing1LimitDegrees=22, swing2LimitDegrees=5, twistLimitDegrees=5}
 {bone1Name=jacket_back_02_l, bone2Name=jacket_back_01_l, swing1LimitDegrees=22, swing2LimitDegrees=5, twistLimitDegrees=5}
 {bone1Name=jacket_back_03_l, bone2Name=jacket_back_02_l, swing1LimitDegrees=22, swing2LimitDegrees=5, twistLimitDegrees=5}
 {bone1Name=jacket_back_04_l, bone2Name=jacket_back_03_l, swing1LimitDegrees=22, swing2LimitDegrees=5, twistLimitDegrees=5}
 {bone1Name=jacket_back_05_l, bone2Name=jacket_back_04_l, swing1LimitDegrees=22, swing2LimitDegrees=5, twistLimitDegrees=5}
 {bone1Name=jacket_side_01_l, bone2Name=jacket_base, swing1LimitDegrees=35, swing2LimitDegrees=35, twistLimitDegrees=5}
 {bone1Name=jacket_side_02_l, bone2Name=jacket_side_01_l, swing1LimitDegrees=35, swing2LimitDegrees=35, twistLimitDegrees=5}
 {bone1Name=jacket_side_03_l, bone2Name=jacket_side_02_l, swing1LimitDegrees=35, swing2LimitDegrees=35, twistLimitDegrees=5}
 {bone1Name=jacket_side_04_l, bone2Name=jacket_side_03_l, swing1LimitDegrees=35, swing2LimitDegrees=35, twistLimitDegrees=5}
 {bone1Name=jacket_side_05_l, bone2Name=jacket_side_04_l, swing1LimitDegrees=35, swing2LimitDegrees=35, twistLimitDegrees=5}
 {bone1Name=jacket_side_06_l, bone2Name=jacket_side_05_l, swing1LimitDegrees=35, swing2LimitDegrees=35, twistLimitDegrees=5}
 {bone1Name=jacket_side_07_l, bone2Name=jacket_side_06_l, swing1LimitDegrees=35, swing2LimitDegrees=35, twistLimitDegrees=5}
 {bone1Name=jacket_side_08_l, bone2Name=jacket_side_07_l, swing1LimitDegrees=35, swing2LimitDegrees=35, twistLimitDegrees=5}
 {bone1Name=jacket_frt_01_l, bone2Name=jacket_base, swing1LimitDegrees=35, swing2LimitDegrees=35, twistLimitDegrees=5}
 {bone1Name=jacket_frt_02_l, bone2Name=jacket_frt_01_l, swing1LimitDegrees=35, swing2LimitDegrees=35, twistLimitDegrees=5}
 {bone1Name=jacket_frt_03_l, bone2Name=jacket_frt_02_l, swing1LimitDegrees=35, swing2LimitDegrees=35, twistLimitDegrees=5}
 {bone1Name=jacket_frt_04_l, bone2Name=jacket_frt_03_l, swing1LimitDegrees=22, swing2LimitDegrees=22, twistLimitDegrees=5}
 {bone1Name=jacket_frt_05_l, bone2Name=jacket_frt_04_l, swing1LimitDegrees=22, swing2LimitDegrees=22, twistLimitDegrees=5}
 {bone1Name=jacket_frt_06_l, bone2Name=jacket_frt_05_l, swing1LimitDegrees=22, swing2LimitDegrees=22, twistLimitDegrees=5}
 {bone1Name=jacket_frt_07_l, bone2Name=jacket_frt_06_l, swing1LimitDegrees=22, swing2LimitDegrees=22, twistLimitDegrees=5}
 {
  bone1Name=calf_l
  bone2Name=thigh_l
  swing1LimitDegrees=45
  swing1Motion=Free
  swing2LimitDegrees=45
  swing2Motion=Free
  twistLimitDegrees=45
  twistMotion=Free
 }
 {bone1Name=jacket_back_01, bone2Name=jacket_base, swing1LimitDegrees=22, swing2LimitDegrees=5, twistLimitDegrees=5}
 {bone1Name=jacket_back_02, bone2Name=jacket_back_01, swing1LimitDegrees=22, swing2LimitDegrees=5, twistLimitDegrees=5}
 {bone1Name=jacket_back_03, bone2Name=jacket_back_02, swing1LimitDegrees=22, swing2LimitDegrees=5, twistLimitDegrees=5}
 {bone1Name=jacket_back_04, bone2Name=jacket_back_03, swing1LimitDegrees=22, swing2LimitDegrees=5, twistLimitDegrees=5}
 {bone1Name=jacket_back_05, bone2Name=jacket_back_04, swing1LimitDegrees=22, swing2LimitDegrees=5, twistLimitDegrees=5}
 {bone1Name=jacket_frtSide_01_l, bone2Name=jacket_base, swing1LimitDegrees=35, swing2LimitDegrees=35, twistLimitDegrees=5}
 {bone1Name=jacket_frtSide_02_l, bone2Name=jacket_frtSide_01_l, swing1LimitDegrees=35, swing2LimitDegrees=35, twistLimitDegrees=5}
 {bone1Name=jacket_frtSide_03_l, bone2Name=jacket_frtSide_02_l, swing1LimitDegrees=22, swing2LimitDegrees=22, twistLimitDegrees=5}
 {bone1Name=jacket_frtSide_04_l, bone2Name=jacket_frtSide_03_l, swing1LimitDegrees=22, swing2LimitDegrees=22, twistLimitDegrees=5}
 {bone1Name=jacket_frtSide_05_l, bone2Name=jacket_frtSide_04_l, swing1LimitDegrees=22, swing2LimitDegrees=22, twistLimitDegrees=5}
 {bone1Name=jacket_back_02, bone2Name=jacket_back_02_l, swing1LimitDegrees=22, swing2LimitDegrees=5, twistLimitDegrees=5}
 {bone1Name=jacket_back_03, bone2Name=jacket_back_03_l, swing1LimitDegrees=22, swing2LimitDegrees=5, twistLimitDegrees=5}
 {bone1Name=jacket_back_04, bone2Name=jacket_back_04_l, swing1LimitDegrees=22, swing2LimitDegrees=5, twistLimitDegrees=5}
 {bone1Name=jacket_back_05, bone2Name=jacket_back_05_l, swing1LimitDegrees=22, swing2LimitDegrees=5, twistLimitDegrees=5}
 {bone1Name=jacket_back_02, bone2Name=jacket_back_02_r, swing1LimitDegrees=22, swing2LimitDegrees=5, twistLimitDegrees=5}
 {bone1Name=jacket_back_03, bone2Name=jacket_back_03_r, swing1LimitDegrees=22, swing2LimitDegrees=5, twistLimitDegrees=5}
 {bone1Name=jacket_back_04, bone2Name=jacket_back_04_r, swing1LimitDegrees=22, swing2LimitDegrees=5, twistLimitDegrees=5}
 {bone1Name=jacket_back_05, bone2Name=jacket_back_05_r, swing1LimitDegrees=22, swing2LimitDegrees=5, twistLimitDegrees=5}
 {bone1Name=thigh_l, bone2Name=pelvis, swing1LimitDegrees=45, swing2LimitDegrees=45, twistLimitDegrees=45}
 {bone1Name=jacket_back_06_r, bone2Name=jacket_back_05_r, swing1LimitDegrees=22, swing2LimitDegrees=5, twistLimitDegrees=5}
 {bone1Name=jacket_back_07_r, bone2Name=jacket_back_06_r, swing1LimitDegrees=22, swing2LimitDegrees=5, twistLimitDegrees=5}
 {bone1Name=jacket_back_06_l, bone2Name=jacket_back_05_l, swing1LimitDegrees=22, swing2LimitDegrees=5, twistLimitDegrees=5}
 {bone1Name=jacket_back_07_l, bone2Name=jacket_back_06_l, swing1LimitDegrees=22, swing2LimitDegrees=5, twistLimitDegrees=5}
 {bone1Name=jacket_frtSide_01_r, bone2Name=jacket_base, swing1LimitDegrees=35, swing2LimitDegrees=35, twistLimitDegrees=5}
 {bone1Name=jacket_frtSide_02_r, bone2Name=jacket_frtSide_01_r, swing1LimitDegrees=35, swing2LimitDegrees=35, twistLimitDegrees=5}
 {bone1Name=jacket_frtSide_03_r, bone2Name=jacket_frtSide_02_r, swing1LimitDegrees=22, swing2LimitDegrees=22, twistLimitDegrees=5}
 {bone1Name=jacket_frtSide_04_r, bone2Name=jacket_frtSide_03_r, swing1LimitDegrees=22, swing2LimitDegrees=22, twistLimitDegrees=5}
 {bone1Name=jacket_frtSide_05_r, bone2Name=jacket_frtSide_04_r, swing1LimitDegrees=22, swing2LimitDegrees=22, twistLimitDegrees=5}
 {bone1Name=jacket_frt_01_r, bone2Name=jacket_base, swing1LimitDegrees=35, swing2LimitDegrees=35, twistLimitDegrees=5}
 {bone1Name=jacket_frt_02_r, bone2Name=jacket_frt_01_r, swing1LimitDegrees=35, swing2LimitDegrees=35, twistLimitDegrees=5}
 {bone1Name=jacket_frt_03_r, bone2Name=jacket_frt_02_r, swing1LimitDegrees=35, swing2LimitDegrees=35, twistLimitDegrees=5}
 {bone1Name=jacket_frt_04_r, bone2Name=jacket_frt_03_r, swing1LimitDegrees=35, swing2LimitDegrees=35, twistLimitDegrees=5}
 {bone1Name=jacket_frt_05_r, bone2Name=jacket_frt_04_r, swing1LimitDegrees=35, swing2LimitDegrees=35, twistLimitDegrees=5}
 {bone1Name=jacket_frt_06_r, bone2Name=jacket_frt_05_r, swing1LimitDegrees=35, swing2LimitDegrees=35, twistLimitDegrees=5}
 {bone1Name=jacket_frt_07_r, bone2Name=jacket_frt_06_r, swing1LimitDegrees=35, swing2LimitDegrees=35, twistLimitDegrees=5}
 {bone1Name=jacket_side_02_r, bone2Name=jacket_side_01_r, swing1LimitDegrees=35, swing2LimitDegrees=35, twistLimitDegrees=5}
 {bone1Name=jacket_side_03_r, bone2Name=jacket_side_02_r, swing1LimitDegrees=35, swing2LimitDegrees=35, twistLimitDegrees=5}
 {bone1Name=jacket_side_04_r, bone2Name=jacket_side_03_r, swing1LimitDegrees=35, swing2LimitDegrees=35, twistLimitDegrees=5}
 {bone1Name=jacket_side_05_r, bone2Name=jacket_side_04_r, swing1LimitDegrees=35, swing2LimitDegrees=35, twistLimitDegrees=5}
 {bone1Name=jacket_side_06_r, bone2Name=jacket_side_05_r, swing1LimitDegrees=35, swing2LimitDegrees=35, twistLimitDegrees=5}
 {bone1Name=jacket_side_07_r, bone2Name=jacket_side_06_r, swing1LimitDegrees=35, swing2LimitDegrees=35, twistLimitDegrees=5}
 {bone1Name=jacket_side_08_r, bone2Name=jacket_side_07_r, swing1LimitDegrees=35, swing2LimitDegrees=35, twistLimitDegrees=5}
 {bone1Name=jacket_side_01_r, bone2Name=jacket_base, swing1LimitDegrees=35, swing2LimitDegrees=35, twistLimitDegrees=5}
 {bone1Name=jacket_frtSide_04_l, bone2Name=jacket_frt_05_l, swing1LimitDegrees=22, swing2LimitDegrees=22, twistLimitDegrees=5}
 {bone1Name=jacket_frtSide_03_l, bone2Name=jacket_frt_04_l, swing1LimitDegrees=22, swing2LimitDegrees=22, twistLimitDegrees=5}
 {bone1Name=jacket_frtSide_05_l, bone2Name=jacket_frt_06_l, swing1LimitDegrees=22, swing2LimitDegrees=22, twistLimitDegrees=5}
 {bone1Name=jacket_frt_08_l, bone2Name=jacket_frt_07_l, swing1LimitDegrees=35, swing2LimitDegrees=35, twistLimitDegrees=5}
 {bone1Name=jacket_frt_08_r, bone2Name=jacket_frt_07_r, swing1LimitDegrees=35, swing2LimitDegrees=35, twistLimitDegrees=5}
 {bone1Name=jacket_frt_09_r, bone2Name=jacket_frt_08_r, swing1LimitDegrees=35, swing2LimitDegrees=35, twistLimitDegrees=5}
 {bone1Name=jacket_frt_09_l, bone2Name=jacket_frt_08_l, swing1LimitDegrees=35, swing2LimitDegrees=35, twistLimitDegrees=5}
]
