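# Animation Blueprint for the StarWalkerFNCS player body that animates its long jacket and collar on a pose copied from the attached parent mesh.
# With rigid bodies enabled, Control Rig feeds RigidBody simulations, with separate setups for front end, skydiving and carried DBNO.
# Otherwise, weighted constraints drive jacket bones from spine, pelvis and legs, with additive bone offsets correcting the shape, some scaled by crouch or calf values.
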AnimBlueprint /Game/Characters/Player/Male/Medium/Bodies/M_MED_StarWalkerFNCS/Meshes/M_MED_StarWalkerFNCS_AnimBP
target_skeleton=&/M_MED_StarWalkerFNCS_Skeleton
graphs [
 AnimGraph {nodes [
   n1 = "Constraint - jacket_frt_03_l (PPPPP)"!(
    ComponentPose=Local To Component(LocalPose=Use cached pose 'AnimCache'())
    Alpha=1
    ConstraintWeights_0=0.1633
    ConstraintWeights_1:float=0.3137
    ConstraintWeights_2:float=0.1434
    ConstraintWeights_3:float=0.3059
    ConstraintWeights_4:float=0.0737
   ) -> (Pose) {
    alphaBoolBlend {blendOption=Linear}
    boneToModify {boneName=jacket_frt_03_l}
    constraintSetup [
     {targetBone {boneName=spine_01}}
     {targetBone {boneName=spine_02}}
     {targetBone {boneName=spine_03}}
     {targetBone {boneName=spine_04}}
     {targetBone {boneName=pelvis}}
    ]
    constraintWeights [1,1,1,1,1]
   }
   n2 = "Constraint - jacket_frt_04_l (PP)"!(ComponentPose=@n1, Alpha=1, ConstraintWeights_0=0.7617, ConstraintWeights_1:float=0.2383) -> (Pose) {
    alphaBoolBlend {blendOption=Linear}
    boneToModify {boneName=jacket_frt_04_l}
    constraintSetup [{targetBone {boneName=pelvis}}, {targetBone {boneName=thigh_l}}]
    constraintWeights [1,1]
   }
   n3 = "Constraint - jacket_frt_05_l (PP)"!(ComponentPose=@n2, Alpha=1, ConstraintWeights_0=0.1011, ConstraintWeights_1:float=0.8989) -> (Pose) {
    alphaBoolBlend {blendOption=Linear}
    boneToModify {boneName=jacket_frt_05_l}
    constraintSetup [{targetBone {boneName=pelvis}}, {targetBone {boneName=thigh_l}}]
    constraintWeights [1,1]
   }
   n4 = "Constraint - jacket_frt_06_l (PP)"!(ComponentPose=@n3, Alpha=1, ConstraintWeights_0=0.0113, ConstraintWeights_1:float=0.9887) -> (Pose) {
    alphaBoolBlend {blendOption=Linear}
    boneToModify {boneName=jacket_frt_06_l}
    constraintSetup [{targetBone {boneName=pelvis}}, {targetBone {boneName=thigh_l}}]
    constraintWeights [1,1]
   }
   n5 = "Constraint - jacket_frt_07_l (PP)"!(ComponentPose=@n4, Alpha=1, ConstraintWeights_0=0.0191, ConstraintWeights_1:float=0.9809) -> (Pose) {
    alphaBoolBlend {blendOption=Linear}
    boneToModify {boneName=jacket_frt_07_l}
    constraintSetup [{targetBone {boneName=calf_l}}, {targetBone {boneName=thigh_l}}]
    constraintWeights [1,1]
   }
   n6 = "Constraint - jacket_frt_08_l (PP)"!(ComponentPose=@n5, Alpha=1, ConstraintWeights_0=0.963, ConstraintWeights_1:float=0.037) -> (Pose) {
    alphaBoolBlend {blendOption=Linear}
    boneToModify {boneName=jacket_frt_08_l}
    constraintSetup [{targetBone {boneName=calf_l}}, {targetBone {boneName=thigh_l}}]
    constraintWeights [1,1]
   }
   n7 = "Constraint - jacket_frt_09_l (P)"!(ComponentPose=@n6, Alpha=1, ConstraintWeights_0=1, ConstraintWeights_1:float=0) -> (Pose) {
    alphaBoolBlend {blendOption=Linear}
    boneToModify {boneName=jacket_frt_09_l}
    constraintSetup [{targetBone {boneName=calf_l}}]
    constraintWeights [1,0]
   }
   n8 = "Constraint - jacket_frt_03_r (PPPPP)"!(
    ComponentPose=@n7
    Alpha=1
    ConstraintWeights_0=0.1633
    ConstraintWeights_1:float=0.3137
    ConstraintWeights_2:float=0.1434
    ConstraintWeights_3:float=0.3059
    ConstraintWeights_4:float=0.0737
   ) -> (Pose) {
    alphaBoolBlend {blendOption=Linear}
    boneToModify {boneName=jacket_frt_03_r}
    constraintSetup [
     {targetBone {boneName=spine_01}}
     {targetBone {boneName=spine_02}}
     {targetBone {boneName=spine_03}}
     {targetBone {boneName=spine_04}}
     {targetBone {boneName=pelvis}}
    ]
    constraintWeights [1,1,1,1,1]
   }
   n9 = "Constraint - jacket_frt_04_r (PP)"!(ComponentPose=@n8, Alpha=1, ConstraintWeights_0=0.2391, ConstraintWeights_1:float=0.7609) -> (Pose) {
    alphaBoolBlend {blendOption=Linear}
    boneToModify {boneName=jacket_frt_04_r}
    constraintSetup [{targetBone {boneName=thigh_r}}, {targetBone {boneName=pelvis}}]
    constraintWeights [1,1]
   }
   n10 = "Constraint - jacket_frt_05_r (PP)"!(ComponentPose=@n9, Alpha=1, ConstraintWeights_0=0.9059, ConstraintWeights_1:float=0.0941) -> (Pose) {
    alphaBoolBlend {blendOption=Linear}
    boneToModify {boneName=jacket_frt_05_r}
    constraintSetup [{targetBone {boneName=thigh_r}}, {targetBone {boneName=pelvis}}]
    constraintWeights [1,1]
   }
   n11 = "Constraint - jacket_frt_06_r (P)"!(ComponentPose=@n10, Alpha=1, ConstraintWeights_0=1, ConstraintWeights_1:float=0) -> (Pose) {
    alphaBoolBlend {blendOption=Linear}
    boneToModify {boneName=jacket_frt_06_r}
    constraintSetup [{targetBone {boneName=thigh_r}}]
    constraintWeights [1,0]
   }
   n12 = "Constraint - jacket_frt_07_r (PP)"!(ComponentPose=@n11, Alpha=1, ConstraintWeights_0=0.9791, ConstraintWeights_1:float=0.0209) -> (Pose) {
    alphaBoolBlend {blendOption=Linear}
    boneToModify {boneName=jacket_frt_07_r}
    constraintSetup [{targetBone {boneName=thigh_r}}, {targetBone {boneName=calf_r}}]
    constraintWeights [1,1]
   }
   n13 = "Constraint - jacket_frt_08_r (PP)"!(ComponentPose=@n12, Alpha=1, ConstraintWeights_0=0.037, ConstraintWeights_1:float=0.963) -> (Pose) {
    alphaBoolBlend {blendOption=Linear}
    boneToModify {boneName=jacket_frt_08_r}
    constraintSetup [{targetBone {boneName=thigh_r}}, {targetBone {boneName=calf_r}}]
    constraintWeights [1,1]
   }
   n14 = "Constraint - jacket_frt_09_r (P)"!(ComponentPose=@n13, Alpha=1, ConstraintWeights_0=1, ConstraintWeights_1:float=0) -> (Pose) {
    alphaBoolBlend {blendOption=Linear}
    boneToModify {boneName=jacket_frt_09_r}
    constraintSetup [{targetBone {boneName=calf_r}}]
    constraintWeights [1,0]
   }
   n15 = "Constraint - jacket_frtSide_02_l (PPPPP)"!(
    ComponentPose=@n14
    Alpha=1
    ConstraintWeights_0=0.2331
    ConstraintWeights_1:float=0.3882
    ConstraintWeights_2:float=0.0505
    ConstraintWeights_3:float=0.1313
    ConstraintWeights_4:float=0.197
   ) -> (Pose) {
    alphaBoolBlend {blendOption=Linear}
    boneToModify {boneName=jacket_frtSide_02_l}
    constraintSetup [
     {targetBone {boneName=spine_01}}
     {targetBone {boneName=spine_02}}
     {targetBone {boneName=spine_03}}
     {targetBone {boneName=spine_04}}
     {targetBone {boneName=pelvis}}
    ]
    constraintWeights [1,1,1,1,1]
   }
   n16 = "Constraint - jacket_frtSide_03_l (PP)"!(ComponentPose=@n15, Alpha=1, ConstraintWeights_0=0.8376, ConstraintWeights_1:float=0.1624) -> (Pose) {
    alphaBoolBlend {blendOption=Linear}
    boneToModify {boneName=jacket_frtSide_03_l}
    constraintSetup [{targetBone {boneName=pelvis}}, {targetBone {boneName=thigh_l}}]
    constraintWeights [1,1]
   }
   n17 = "Constraint - jacket_frtSide_04_l (PP)"!(ComponentPose=@n16, Alpha=1, ConstraintWeights_0=0.1335, ConstraintWeights_1:float=0.8665) -> (Pose) {
    alphaBoolBlend {blendOption=Linear}
    boneToModify {boneName=jacket_frtSide_04_l}
    constraintSetup [{targetBone {boneName=pelvis}}, {targetBone {boneName=thigh_l}}]
    constraintWeights [1,1]
   }
   n18 = "Constraint - jacket_frtSide_05_l (PP)"!(ComponentPose=@n17, Alpha=1, ConstraintWeights_0=0.0134, ConstraintWeights_1:float=0.9866) -> (Pose) {
    alphaBoolBlend {blendOption=Linear}
    boneToModify {boneName=jacket_frtSide_05_l}
    constraintSetup [{targetBone {boneName=pelvis}}, {targetBone {boneName=thigh_l}}]
    constraintWeights [1,1]
   }
   n19 = "Constraint - jacket_frtSide_02_r (PPPPP)"!(
    ComponentPose=@n18
    Alpha=1
    ConstraintWeights_0=0.2289
    ConstraintWeights_1:float=0.3815
    ConstraintWeights_2:float=0.0525
    ConstraintWeights_3:float=0.1321
    ConstraintWeights_4:float=0.2049
   ) -> (Pose) {
    alphaBoolBlend {blendOption=Linear}
    boneToModify {boneName=jacket_frtSide_02_r}
    constraintSetup [
     {targetBone {boneName=spine_01}}
     {targetBone {boneName=spine_02}}
     {targetBone {boneName=spine_03}}
     {targetBone {boneName=spine_04}}
     {targetBone {boneName=pelvis}}
    ]
    constraintWeights [1,1,1,1,1]
   }
   n20 = "Constraint - jacket_frtSide_03_r (PP)"!(ComponentPose=@n19, Alpha=1, ConstraintWeights_0=0.1853, ConstraintWeights_1:float=0.8147) -> (Pose) {
    alphaBoolBlend {blendOption=Linear}
    boneToModify {boneName=jacket_frtSide_03_r}
    constraintSetup [{targetBone {boneName=thigh_r}}, {targetBone {boneName=pelvis}}]
    constraintWeights [1,1]
   }
   n21 = "Constraint - jacket_frtSide_04_r (PP)"!(ComponentPose=@n20, Alpha=1, ConstraintWeights_0=0.8764, ConstraintWeights_1:float=0.1236) -> (Pose) {
    alphaBoolBlend {blendOption=Linear}
    boneToModify {boneName=jacket_frtSide_04_r}
    constraintSetup [{targetBone {boneName=thigh_r}}, {targetBone {boneName=pelvis}}]
    constraintWeights [1,1]
   }
   n22 = "Constraint - jacket_frtSide_05_r (P)"!(ComponentPose=@n21, Alpha=1, ConstraintWeights_0=1, ConstraintWeights_1:float=0) -> (Pose) {
    alphaBoolBlend {blendOption=Linear}
    boneToModify {boneName=jacket_frtSide_05_r}
    constraintSetup [{targetBone {boneName=thigh_r}}]
    constraintWeights [1,0]
   }
   n23 = "Constraint - jacket_side_02_l (PPPPP)"!(
    ComponentPose=@n22
    Alpha=1
    ConstraintWeights_0=0.2426
    ConstraintWeights_1:float=0.334
    ConstraintWeights_2:float=0.0344
    ConstraintWeights_3:float=0.0663
    ConstraintWeights_4:float=0.3227
   ) -> (Pose) {
    alphaBoolBlend {blendOption=Linear}
    boneToModify {boneName=jacket_side_02_l}
    constraintSetup [
     {targetBone {boneName=spine_01}}
     {targetBone {boneName=spine_02}}
     {targetBone {boneName=spine_03}}
     {targetBone {boneName=spine_04}}
     {targetBone {boneName=pelvis}}
    ]
    constraintWeights [1,1,1,1,1]
   }
   n24 = "Constraint - jacket_side_03_l (PP)"!(ComponentPose=@n23, Alpha=1, ConstraintWeights_0=0.8489, ConstraintWeights_1:float=0.1511) -> (Pose) {
    alphaBoolBlend {blendOption=Linear}
    boneToModify {boneName=jacket_side_03_l}
    constraintSetup [{targetBone {boneName=pelvis}}, {targetBone {boneName=thigh_l}}]
    constraintWeights [1,1]
   }
   n25 = "Constraint - jacket_side_04_l (PP)"!(ComponentPose=@n24, Alpha=1, ConstraintWeights_0=0.3356, ConstraintWeights_1:float=0.6644) -> (Pose) {
    alphaBoolBlend {blendOption=Linear}
    boneToModify {boneName=jacket_side_04_l}
    constraintSetup [{targetBone {boneName=pelvis}}, {targetBone {boneName=thigh_l}}]
    constraintWeights [1,1]
   }
   n26 = "Constraint - jacket_side_05_l (PP)"!(ComponentPose=@n25, Alpha=1, ConstraintWeights_0=0.0309, ConstraintWeights_1:float=0.9691) -> (Pose) {
    alphaBoolBlend {blendOption=Linear}
    boneToModify {boneName=jacket_side_05_l}
    constraintSetup [{targetBone {boneName=pelvis}}, {targetBone {boneName=thigh_l}}]
    constraintWeights [1,1]
   }
   n27 = "Constraint - jacket_side_06_l (P)"!(ComponentPose=@n26, Alpha=1, ConstraintWeights_0=1, ConstraintWeights_1:float=0) -> (Pose) {
    alphaBoolBlend {blendOption=Linear}
    boneToModify {boneName=jacket_side_06_l}
    constraintSetup [{targetBone {boneName=thigh_l}}]
    constraintWeights [1,0]
   }
   n28 = "Constraint - jacket_side_07_l (PP)"!(ComponentPose=@n27, Alpha=1, ConstraintWeights_0=0.3489, ConstraintWeights_1:float=0.6511) -> (Pose) {
    alphaBoolBlend {blendOption=Linear}
    boneToModify {boneName=jacket_side_07_l}
    constraintSetup [{targetBone {boneName=calf_l}}, {targetBone {boneName=thigh_l}}]
    constraintWeights [1,1]
   }
   n29 = "Constraint - jacket_side_08_l (PP)"!(ComponentPose=@n28, Alpha=1, ConstraintWeights_0=0.974, ConstraintWeights_1:float=0.026) -> (Pose) {
    alphaBoolBlend {blendOption=Linear}
    boneToModify {boneName=jacket_side_08_l}
    constraintSetup [{targetBone {boneName=calf_l}}, {targetBone {boneName=thigh_l}}]
    constraintWeights [1,1]
   }
   n30 = "Constraint - jacket_side_02_r (PPPPP)"!(
    ComponentPose=@n29
    Alpha=1
    ConstraintWeights_0=0.2507
    ConstraintWeights_1:float=0.3439
    ConstraintWeights_2:float=0.0296
    ConstraintWeights_3:float=0.0635
    ConstraintWeights_4:float=0.3122
   ) -> (Pose) {
    alphaBoolBlend {blendOption=Linear}
    boneToModify {boneName=jacket_side_02_r}
    constraintSetup [
     {targetBone {boneName=spine_01}}
     {targetBone {boneName=spine_02}}
     {targetBone {boneName=spine_03}}
     {targetBone {boneName=spine_04}}
     {targetBone {boneName=pelvis}}
    ]
    constraintWeights [1,1,1,1,1]
   }
   n31 = "Constraint - jacket_side_03_r (PP)"!(ComponentPose=@n30, Alpha=1, ConstraintWeights_0=0.1511, ConstraintWeights_1:float=0.8489) -> (Pose) {
    alphaBoolBlend {blendOption=Linear}
    boneToModify {boneName=jacket_side_03_r}
    constraintSetup [{targetBone {boneName=thigh_r}}, {targetBone {boneName=pelvis}}]
    constraintWeights [1,1]
   }
   n32 = "Constraint - jacket_side_04_r (PP)"!(ComponentPose=@n31, Alpha=1, ConstraintWeights_0=0.6682, ConstraintWeights_1:float=0.3318) -> (Pose) {
    alphaBoolBlend {blendOption=Linear}
    boneToModify {boneName=jacket_side_04_r}
    constraintSetup [{targetBone {boneName=thigh_r}}, {targetBone {boneName=pelvis}}]
    constraintWeights [1,1]
   }
   n33 = "Constraint - jacket_side_05_r (PP)"!(ComponentPose=@n32, Alpha=1, ConstraintWeights_0=0.9865, ConstraintWeights_1:float=0.0135) -> (Pose) {
    alphaBoolBlend {blendOption=Linear}
    boneToModify {boneName=jacket_side_05_r}
    constraintSetup [{targetBone {boneName=thigh_r}}, {targetBone {boneName=pelvis}}]
    constraintWeights [1,1]
   }
   n34 = "Constraint - jacket_side_06_r (P)"!(ComponentPose=@n33, Alpha=1, ConstraintWeights_0=1, ConstraintWeights_1:float=0) -> (Pose) {
    alphaBoolBlend {blendOption=Linear}
    boneToModify {boneName=jacket_side_06_r}
    constraintSetup [{targetBone {boneName=thigh_r}}]
    constraintWeights [1,0]
   }
   n35 = "Constraint - jacket_side_07_r (PP)"!(ComponentPose=@n34, Alpha=1, ConstraintWeights_0=0.6415, ConstraintWeights_1:float=0.3584) -> (Pose) {
    alphaBoolBlend {blendOption=Linear}
    boneToModify {boneName=jacket_side_07_r}
    constraintSetup [{targetBone {boneName=thigh_r}}, {targetBone {boneName=calf_r}}]
    constraintWeights [1,1]
   }
   n36 = "Constraint - jacket_side_08_r (PP)"!(ComponentPose=@n35, Alpha=1, ConstraintWeights_0=0.026, ConstraintWeights_1:float=0.974) -> (Pose) {
    alphaBoolBlend {blendOption=Linear}
    boneToModify {boneName=jacket_side_08_r}
    constraintSetup [{targetBone {boneName=thigh_r}}, {targetBone {boneName=calf_r}}]
    constraintWeights [1,1]
   }
   n37 = "Constraint - jacket_back_02_l (PPPPP)"!(
    ComponentPose=@n36
    Alpha=1
    ConstraintWeights_0=0.2786
    ConstraintWeights_1:float=0.3325
    ConstraintWeights_2:float=0.0691
    ConstraintWeights_3:float=0.026
    ConstraintWeights_4:float=0.2939
   ) -> (Pose) {
    alphaBoolBlend {blendOption=Linear}
    boneToModify {boneName=jacket_back_02_l}
    constraintSetup [
     {targetBone {boneName=spine_01}}
     {targetBone {boneName=spine_02}}
     {targetBone {boneName=spine_03}}
     {targetBone {boneName=spine_04}}
     {targetBone {boneName=pelvis}}
    ]
    constraintWeights [1,1,1,1,1]
   }
   n38 = "Constraint - jacket_back_03_l (PP)"!(ComponentPose=@n37, Alpha=1, ConstraintWeights_0=0.9601, ConstraintWeights_1:float=0.0399) -> (Pose) {
    alphaBoolBlend {blendOption=Linear}
    boneToModify {boneName=jacket_back_03_l}
    constraintSetup [{targetBone {boneName=pelvis}}, {targetBone {boneName=thigh_l}}]
    constraintWeights [1,1]
   }
   n39 = "Constraint - jacket_back_04_l (PP)"!(ComponentPose=@n38, Alpha=1, ConstraintWeights_0=0.5447, ConstraintWeights_1:float=0.4553) -> (Pose) {
    alphaBoolBlend {blendOption=Linear}
    boneToModify {boneName=jacket_back_04_l}
    constraintSetup [{targetBone {boneName=pelvis}}, {targetBone {boneName=thigh_l}}]
    constraintWeights [1,1]
   }
   n40 = "Constraint - jacket_back_05_l (PP)"!(ComponentPose=@n39, Alpha=1, ConstraintWeights_0=0.0682, ConstraintWeights_1:float=0.9318) -> (Pose) {
    alphaBoolBlend {blendOption=Linear}
    boneToModify {boneName=jacket_back_05_l}
    constraintSetup [{targetBone {boneName=pelvis}}, {targetBone {boneName=thigh_l}}]
    constraintWeights [1,1]
   }
   n41 = "Constraint - jacket_back_02_r (PPPPP)"!(
    ComponentPose=@n40
    Alpha=1
    ConstraintWeights_0=0.283
    ConstraintWeights_1:float=0.3464
    ConstraintWeights_2:float=0.072
    ConstraintWeights_3:float=0.0211
    ConstraintWeights_4:float=0.2775
   ) -> (Pose) {
    alphaBoolBlend {blendOption=Linear}
    boneToModify {boneName=jacket_back_02_r}
    constraintSetup [
     {targetBone {boneName=spine_01}}
     {targetBone {boneName=spine_02}}
     {targetBone {boneName=spine_03}}
     {targetBone {boneName=spine_04}}
     {targetBone {boneName=pelvis}}
    ]
    constraintWeights [1,1,1,1,1]
   }
   n42 = "Constraint - jacket_back_03_r (PP)"!(ComponentPose=@n41, Alpha=1, ConstraintWeights_0=0.0399, ConstraintWeights_1:float=0.9601) -> (Pose) {
    alphaBoolBlend {blendOption=Linear}
    boneToModify {boneName=jacket_back_03_r}
    constraintSetup [{targetBone {boneName=thigh_r}}, {targetBone {boneName=pelvis}}]
    constraintWeights [1,1]
   }
   n43 = "Constraint - jacket_back_04_r (PP)"!(ComponentPose=@n42, Alpha=1, ConstraintWeights_0=0.4565, ConstraintWeights_1:float=0.5435) -> (Pose) {
    alphaBoolBlend {blendOption=Linear}
    boneToModify {boneName=jacket_back_04_r}
    constraintSetup [{targetBone {boneName=thigh_r}}, {targetBone {boneName=pelvis}}]
    constraintWeights [1,1]
   }
   n44 = "Constraint - jacket_back_05_r (PP)"!(ComponentPose=@n43, Alpha=1, ConstraintWeights_0=0.9486, ConstraintWeights_1:float=0.0514) -> (Pose) {
    alphaBoolBlend {blendOption=Linear}
    boneToModify {boneName=jacket_back_05_r}
    constraintSetup [{targetBone {boneName=thigh_r}}, {targetBone {boneName=pelvis}}]
    constraintWeights [1,1]
   }
   n45 = "Constraint - jacket_back_02 (PPPPP)"!(
    ComponentPose=@n44
    Alpha=1
    ConstraintWeights_0=0.2641
    ConstraintWeights_1:float=0.3737
    ConstraintWeights_2:float=0.1147
    ConstraintWeights_3:float=0.0276
    ConstraintWeights_4:float=0.2199
   ) -> (Pose) {
    alphaBoolBlend {blendOption=Linear}
    boneToModify {boneName=jacket_back_02}
    constraintSetup [
     {targetBone {boneName=spine_01}}
     {targetBone {boneName=spine_02}}
     {targetBone {boneName=spine_03}}
     {targetBone {boneName=spine_04}}
     {targetBone {boneName=pelvis}}
    ]
    constraintWeights [1,1,1,1,1]
   }
   n46 = "Constraint - jacket_back_03 (P)"!(ComponentPose=@n45, Alpha=1, ConstraintWeights_0=1, ConstraintWeights_1:float=0) -> (Pose) {
    alphaBoolBlend {blendOption=Linear}
    boneToModify {boneName=jacket_back_03}
    constraintSetup [{targetBone {boneName=pelvis}}]
    constraintWeights [1,0]
   }
   n47 = "Constraint - jacket_back_04 (PPP)"!(ComponentPose=@n46, Alpha=1, ConstraintWeights_0=0.1801, ConstraintWeights_1:float=0.7586, ConstraintWeights_2:float=0.0613) -> (Pose) {
    alphaBoolBlend {blendOption=Linear}
    boneToModify {boneName=jacket_back_04}
    constraintSetup [{targetBone {boneName=thigh_r}}, {targetBone {boneName=pelvis}}, {targetBone {boneName=thigh_l}}]
    constraintWeights [1,1,1]
   }
   n48 = "Constraint - jacket_back_05 (PP)"!(ComponentPose=@n47, Alpha=1, ConstraintWeights_0=0.0699, ConstraintWeights_1:float=0.9301) -> (Pose) {
    alphaBoolBlend {blendOption=Linear}
    boneToModify {boneName=jacket_back_05}
    constraintSetup [{targetBone {boneName=pelvis}}, {targetBone {boneName=thigh_l}}]
    constraintWeights [1,1]
   }
   n49 = "Transform (Modify) Bone - Bone: jacket_back_04_l"(ComponentPose=@n48, Alpha=1, Translation="0,0,0", Rotation="0,-20,0", Scale="1,1,1") {
    rotationMode=BMM_Additive
    rotationSpace=BCS_BoneSpace
    alphaBoolBlend {blendOption=Linear}
    boneToModify {boneName=jacket_back_04_l}
    rotation {yaw=-20}
   }
   n50 = "Transform (Modify) Bone - Bone: jacket_back_04_r"(ComponentPose=@n49, Alpha=1, Translation="0,0,0", Rotation="0,-15,0", Scale="1,1,1") {
    rotationMode=BMM_Additive
    rotationSpace=BCS_BoneSpace
    alphaBoolBlend {blendOption=Linear}
    boneToModify {boneName=jacket_back_04_r}
    rotation {yaw=-15}
   }
   n51 = "Transform (Modify) Bone - Bone: jacket_back_06_l"(ComponentPose=@n50, Alpha=Get calf_rot_back_l() -> (calf_rot_back_l:double), Translation="0,0,0", Rotation="0,-35,0", Scale="1,1,1") {
    rotationMode=BMM_Additive
    rotationSpace=BCS_BoneSpace
    alphaBoolBlend {blendOption=Linear}
    boneToModify {boneName=jacket_back_06_l}
    rotation {yaw=-35}
   }
   n52 = "Transform (Modify) Bone - Bone: jacket_back_06_r"(ComponentPose=@n51, Alpha=Get calf_rot_back_r() -> (calf_rot_back_r:double), Translation="0,0,0", Rotation="0,-35,0", Scale="1,1,1") {
    rotationMode=BMM_Additive
    rotationSpace=BCS_BoneSpace
    alphaBoolBlend {blendOption=Linear}
    boneToModify {boneName=jacket_back_06_r}
    rotation {yaw=-35}
   }
   n53 = "Transform (Modify) Bone - Bone: jacket_frt_03_r"(ComponentPose=@n52, Alpha=1, Translation="0,3,0", Rotation="0,0,0", Scale="1,1,1") {
    rotationMode=BMM_Additive
    rotationSpace=BCS_BoneSpace
    translationMode=BMM_Additive
    translationSpace=BCS_BoneSpace
    alphaBoolBlend {blendOption=Linear}
    alphaScaleBiasClamp {bInterpResult=true, interpSpeedIncreasing=2}
    boneToModify {boneName=jacket_frt_03_r}
    translation {y=3}
   }
   n54 = "Transform (Modify) Bone - Bone: jacket_frt_03_l"(ComponentPose=@n53, Alpha=1, Translation="0,3,0", Rotation="0,0,0", Scale="1,1,1") {
    rotationMode=BMM_Additive
    rotationSpace=BCS_BoneSpace
    translationMode=BMM_Additive
    translationSpace=BCS_BoneSpace
    alphaBoolBlend {blendOption=Linear}
    alphaScaleBiasClamp {bInterpResult=true, interpSpeedIncreasing=2}
    boneToModify {boneName=jacket_frt_03_l}
    translation {y=3}
   }
   n55 = "Transform (Modify) Bone - Bone: jacket_side_02_r"(ComponentPose=@n54, Alpha=1, Translation="0,3,0", Rotation="0,0,0", Scale="1,1,1") {
    rotationMode=BMM_Additive
    rotationSpace=BCS_BoneSpace
    translationMode=BMM_Additive
    translationSpace=BCS_BoneSpace
    alphaBoolBlend {blendOption=Linear}
    alphaScaleBiasClamp {bInterpResult=true, interpSpeedDecreasing=5}
    boneToModify {boneName=jacket_side_02_r}
    translation {y=3}
   }
   n56 = "Transform (Modify) Bone - Bone: jacket_side_02_l"(ComponentPose=@n55, Alpha=1, Translation="0,3,0", Rotation="0,0,0", Scale="1,1,1") {
    rotationMode=BMM_Additive
    rotationSpace=BCS_BoneSpace
    translationMode=BMM_Additive
    translationSpace=BCS_BoneSpace
    alphaBoolBlend {blendOption=Linear}
    alphaScaleBiasClamp {bInterpResult=true, interpSpeedDecreasing=5}
    boneToModify {boneName=jacket_side_02_l}
    translation {y=3}
   }
   n57 = "Transform (Modify) Bone - Bone: jacket_frtSide_02_r"(ComponentPose=@n56, Alpha=1, Translation="0,3,0", Rotation="0,0,0", Scale="1,1,1") {
    rotationMode=BMM_Additive
    rotationSpace=BCS_BoneSpace
    translationMode=BMM_Additive
    translationSpace=BCS_BoneSpace
    alphaBoolBlend {blendOption=Linear}
    alphaScaleBiasClamp {bInterpResult=true, interpSpeedDecreasing=5}
    boneToModify {boneName=jacket_frtSide_02_r}
    translation {y=3}
   }
   n58 = "Transform (Modify) Bone - Bone: jacket_frtSide_02_l"(ComponentPose=@n57, Alpha=1, Translation="0,3,0", Rotation="0,0,0", Scale="1,1,1") {
    rotationMode=BMM_Additive
    rotationSpace=BCS_BoneSpace
    translationMode=BMM_Additive
    translationSpace=BCS_BoneSpace
    alphaBoolBlend {blendOption=Linear}
    alphaScaleBiasClamp {bInterpResult=true, interpSpeedDecreasing=5}
    boneToModify {boneName=jacket_frtSide_02_l}
    translation {y=3}
   }
   n59 = Get Is Crouching Float() -> (Is Crouching Float:double)
   n60 = "Transform (Modify) Bone - Bone: jacket_frt_02_r"(ComponentPose=@n58, Alpha=@n59, Translation="7,0,1", Rotation="10,0,0", Scale="1,1,1") {
    rotationMode=BMM_Additive
    rotationSpace=BCS_BoneSpace
    translationMode=BMM_Additive
    translationSpace=BCS_BoneSpace
    alphaBoolBlend {blendOption=Linear}
    alphaScaleBiasClamp {bInterpResult=true, interpSpeedDecreasing=5}
    boneToModify {boneName=jacket_frt_02_r}
    rotation {pitch=10}
    translation {x=7, z=1}
   }
   n61 = "Transform (Modify) Bone - Bone: jacket_frt_02_l"(ComponentPose=@n60, Alpha=@n59, Translation="7,0,-1", Rotation="10,0,0", Scale="1,1,1") {
    rotationMode=BMM_Additive
    rotationSpace=BCS_BoneSpace
    translationMode=BMM_Additive
    translationSpace=BCS_BoneSpace
    alphaBoolBlend {blendOption=Linear}
    alphaScaleBiasClamp {bInterpResult=true, interpSpeedDecreasing=5}
    boneToModify {boneName=jacket_frt_02_l}
    rotation {pitch=10}
    translation {x=7, z=-1}
   }
   n62 = "Transform (Modify) Bone - Bone: jacket_frtSide_02_l"(ComponentPose=@n61, Alpha=@n59, Translation="10,1,0", Rotation="0,0,0", Scale="1,1,1") {
    rotationMode=BMM_Additive
    rotationSpace=BCS_BoneSpace
    translationMode=BMM_Additive
    translationSpace=BCS_BoneSpace
    alphaBoolBlend {blendOption=Linear}
    alphaScaleBiasClamp {bInterpResult=true, interpSpeedDecreasing=5}
    boneToModify {boneName=jacket_frtSide_02_l}
    translation=<10 1>
   }
   n63 = "Transform (Modify) Bone - Bone: jacket_frtSide_02_r"(ComponentPose=@n62, Alpha=@n59, Translation="10,1,0", Rotation="0,0,0", Scale="1,1,1") {
    rotationMode=BMM_Additive
    rotationSpace=BCS_BoneSpace
    translationMode=BMM_Additive
    translationSpace=BCS_BoneSpace
    alphaBoolBlend {blendOption=Linear}
    alphaScaleBiasClamp {bInterpResult=true, interpSpeedDecreasing=5}
    boneToModify {boneName=jacket_frtSide_02_r}
    translation=<10 1>
   }
   n64 = "Transform (Modify) Bone - Bone: jacket_side_02_l"(ComponentPose=@n63, Alpha=@n59, Translation="1,0,0", Rotation="0,0,0", Scale="1,1,1") {
    rotationMode=BMM_Additive
    rotationSpace=BCS_BoneSpace
    translationMode=BMM_Additive
    translationSpace=BCS_BoneSpace
    alphaBoolBlend {blendOption=Linear}
    alphaScaleBiasClamp {bInterpResult=true, interpSpeedDecreasing=5}
    boneToModify {boneName=jacket_side_02_l}
    translation {x=1}
   }
   n65 = "Transform (Modify) Bone - Bone: jacket_side_02_r"(ComponentPose=@n64, Alpha=@n59, Translation="1,0,0", Rotation="0,0,0", Scale="1,1,1") {
    rotationMode=BMM_Additive
    rotationSpace=BCS_BoneSpace
    translationMode=BMM_Additive
    translationSpace=BCS_BoneSpace
    alphaBoolBlend {blendOption=Linear}
    alphaScaleBiasClamp {bInterpResult=true, interpSpeedDecreasing=5}
    boneToModify {boneName=jacket_side_02_r}
    translation {x=1}
   }
   n66 = "Transform (Modify) Bone - Bone: jacket_frt_06_r"(ComponentPose=@n65, Alpha=@n59, Translation="0,0,5", Rotation="0,0,0", Scale="1,1,1") {
    rotationMode=BMM_Additive
    rotationSpace=BCS_BoneSpace
    translationMode=BMM_Additive
    translationSpace=BCS_BoneSpace
    alphaBoolBlend {blendOption=Linear}
    alphaScaleBiasClamp {bInterpResult=true, interpSpeedDecreasing=5}
    boneToModify {boneName=jacket_frt_06_r}
    translation {z=5}
   }
   n67 = "Transform (Modify) Bone - Bone: jacket_frt_06_l"(ComponentPose=@n66, Alpha=@n59, Translation="0,0,-5", Rotation="0,0,0", Scale="1,1,1") {
    rotationMode=BMM_Additive
    rotationSpace=BCS_BoneSpace
    translationMode=BMM_Additive
    translationSpace=BCS_BoneSpace
    alphaBoolBlend {blendOption=Linear}
    alphaScaleBiasClamp {bInterpResult=true, interpSpeedDecreasing=5}
    boneToModify {boneName=jacket_frt_06_l}
    translation {z=-5}
   }
   n68 = Component To Local(ComponentPose=@n67)
   n69 = Control Rig!(
    Source=Apply Additive(
     Base=Use cached pose 'AnimCache'()
     Additive="Sequence Player (additive) 'M_MED_Star_Walker_Reset_Pose'"() {sequence=/Game/Characters/Player/Male/Medium/Bodies/M_MED_Star_Walker/Meshes/M_MED_Star_Walker_Reset_Pose}
     Alpha=0.8
    ) {alphaBoolBlend {blendOption=Linear}}
    Alpha
    Is Rifle Alpha:double=Get Is Rifle Alpha() -> (Is Rifle Alpha:double)
    Is Crouching Float:double=Get Is Crouching Float()
    Is Pistol Alpha:double=Get Is Pistol Alpha() -> (Is Pistol Alpha:double)
   ) -> (Pose) {
    alphaBoolBlend {blendOption=Linear}
    controlRigAssetReference {blueprintRigClass=&/M_MED_StarWalkerFNCS_CtrlRig.M_MED_StarWalkerFNCS_CtrlRig_C}
   }
   n70 = RigidBody!(
    ComponentPose=Local To Component(LocalPose=@n69)
    Alpha=Get CoatAlpha() -> (CoatAlpha:double)
    OverrideWorldGravity:Vector=Get GravityOverride()
    ExternalForce
   ) -> (Pose) {
    bClampLinearTranslationLimitToRefPose=true
    bOverrideWorldGravity=true
    overridePhysicsAsset=&/M_MED_StarWalkerFNCS_Coat_Physics
    simulationSpace=BaseBoneSpace
    alphaBoolBlend {blendOption=Linear}
    alphaScaleBiasClamp {bInterpResult=true, interpSpeedIncreasing=2}
    baseBoneRef {boneName=pelvis}
    componentAppliedLinearAccClamp=<3000 3000 3000>
    componentLinearAccScale=<0.4 0.3 0.3>
    componentLinearVelScale=<0.4 0.3 0.3>
    simSpaceSettings {
     maxAngularAcceleration=1000
     maxAngularVelocity=3
     maxLinearAcceleration=700
     maxLinearVelocity=500
     velocityScaleZ=0
     worldAlpha=0.15
     externalAngularVelocity {z=0.1}
     externalLinearDragV=<0.2 0.2>
    }
   }
   n71 = RigidBody!(ComponentPose=@n70, Alpha, ExternalForce) -> (Pose) {
    bClampLinearTranslationLimitToRefPose=true
    bOverrideWorldGravity=true
    overridePhysicsAsset=&/M_MED_StarWalkerFNCS_Collar_Physics
    simulationSpace=BaseBoneSpace
    alphaBoolBlend {blendOption=Linear}
    baseBoneRef {boneName=spine_05}
    componentAppliedLinearAccClamp=<1000 1000 1300>
    componentLinearAccScale=<0.8 0.8 0.6>
    componentLinearVelScale=<0.8 0.8 0.6>
    overrideWorldGravity {z=-950}
   }
   n72 = Get Is Playing Emote Float() -> (Is Playing Emote Float:double)
   n73 = "Transform (Modify) Bone - Bone: jacket_frt_01_r"(ComponentPose=@n71, Alpha=@n72, Translation="0,0,0", Rotation="-15,-10,0", Scale="1,1,1") {
    rotationMode=BMM_Additive
    rotationSpace=BCS_BoneSpace
    translationMode=BMM_Additive
    translationSpace=BCS_BoneSpace
    alphaBoolBlend {blendOption=Linear}
    alphaScaleBiasClamp {bInterpResult=true, interpSpeedIncreasing=2}
    boneToModify {boneName=jacket_frt_01_r}
    rotation {pitch=-15, yaw=-10}
   }
   n74 = "Transform (Modify) Bone - Bone: jacket_frt_01_l"(ComponentPose=@n73, Alpha=@n72, Translation="0,0,0", Rotation="15,-10,0", Scale="1,1,1") {
    rotationMode=BMM_Additive
    rotationSpace=BCS_BoneSpace
    translationMode=BMM_Additive
    translationSpace=BCS_BoneSpace
    alphaBoolBlend {blendOption=Linear}
    alphaScaleBiasClamp {bInterpResult=true, interpSpeedIncreasing=2}
    boneToModify {boneName=jacket_frt_01_l}
    rotation {pitch=15, yaw=-10}
   }
   n75 = Component To Local(ComponentPose=@n74)
   n76 = Control Rig(
    Source=Use cached pose 'AnimCache'()
    Is Rifle Alpha=Get Is Rifle Alpha()
    Is Crouching Float=Get Is Crouching Float()
    Is Pistol Alpha=Get Is Pistol Alpha()
   ) {
    alphaBoolBlend {blendOption=Linear}
    controlRigAssetReference {blueprintRigClass=&M_MED_StarWalkerFNCS_CtrlRig_C}
   }
   n77 = RigidBody!(
    ComponentPose=Local To Component(LocalPose=@n76)
    Alpha
    OverrideWorldGravity:Vector=Get GravityOverride()
    ExternalForce=Get skydiving_force() -> (skydiving_force:Vector)
   ) -> (Pose) {
    bClampLinearTranslationLimitToRefPose=true
    bOverrideWorldGravity=true
    overridePhysicsAsset=&/M_MED_StarWalkerFNCS_Skydive_Coat_Physics
    simulationSpace=BaseBoneSpace
    alphaBoolBlend {blendOption=Linear}
    baseBoneRef {boneName=pelvis}
    componentAppliedLinearAccClamp=<900 600 600>
    componentLinearAccScale=<0.1 0.1 0.1>
    componentLinearVelScale=<0.1 0.1 0.1>
    overrideWorldGravity {z=-150}
   }
   n78 = "Transform (Modify) Bone - Bone: jacket_side_01_l"(ComponentPose=@n77, Alpha=1, Translation="0,0,0", Rotation="0,-5,0", Scale="1,1,1") {
    rotationMode=BMM_Additive
    rotationSpace=BCS_BoneSpace
    translationMode=BMM_Additive
    translationSpace=BCS_BoneSpace
    alphaBoolBlend {blendOption=Linear}
    boneToModify {boneName=jacket_side_01_l}
    rotation {yaw=-5}
   }
   n79 = "Transform (Modify) Bone - Bone: jacket_side_01_r"(ComponentPose=@n78, Alpha=1, Translation="0,0,0", Rotation="0,-5,0", Scale="1,1,1") {
    rotationMode=BMM_Additive
    rotationSpace=BCS_BoneSpace
    translationMode=BMM_Additive
    translationSpace=BCS_BoneSpace
    alphaBoolBlend {blendOption=Linear}
    boneToModify {boneName=jacket_side_01_r}
    rotation {yaw=-5}
   }
   n80 = Component To Local(ComponentPose=@n79)
   n81 = Get bIsSkyDiving()
   n82 = RigidBody!(
    ComponentPose=Local To Component(LocalPose=Blend Poses by bool(BlendPose_0=@n80, BlendPose_1=@n75, bActiveValue=@n81) {blendType=Linear})
    Alpha=Get Is DBNOCarried() -> (Is DBNOCarried:double)
    ExternalForce
   ) -> (Pose) {
    bClampLinearTranslationLimitToRefPose=true
    bOverrideWorldGravity=true
    overridePhysicsAsset=/Game/Characters/Player/Male/Medium/Bodies/M_MED_Star_Walker/Meshes/M_MED_Star_Walker_Coat_Physics_DBNOcarried
    simulationSpace=BaseBoneSpace
    alphaBoolBlend {blendOption=Linear}
    baseBoneRef {boneName=root}
    componentAppliedLinearAccClamp=<1000 1000 1000>
    componentLinearAccScale=<0.6 0.6 0.1>
    componentLinearVelScale=<0.6 0.6 0.1>
    overrideWorldGravity {z=-725}
   }
   n83 = Component To Local(ComponentPose=@n82)
   n84 = Get bRigidBodyEnabled()
   n85 = Control Rig(
    Source=Apply Additive(
     Base=Use cached pose 'AnimCache'()
     Additive="Sequence Player (additive) 'M_MED_Star_Walker_Reset_Pose'"() {sequence=&M_MED_Star_Walker_Reset_Pose}
     Alpha=0.8
    ) {alphaBoolBlend {blendOption=Linear}}
    Is Rifle Alpha=Get Is Rifle Alpha()
    Is Crouching Float=Get Is Crouching Float()
    Is Pistol Alpha=Get Is Pistol Alpha()
   ) {
    alphaBoolBlend {blendOption=Linear}
    controlRigAssetReference {blueprintRigClass=&M_MED_StarWalkerFNCS_CtrlRig_C}
   }
   n86 = RigidBody!(ComponentPose=Local To Component(LocalPose=@n85), Alpha=Get CoatAlpha(), OverrideWorldGravity:Vector=Get GravityOverride(), ExternalForce) -> (Pose) {
    bClampLinearTranslationLimitToRefPose=true
    bOverrideWorldGravity=true
    overridePhysicsAsset=&/M_MED_StarWalkerFNCS_FECoat_Physics
    simulationSpace=BaseBoneSpace
    alphaBoolBlend {blendOption=Linear}
    alphaScaleBiasClamp {bInterpResult=true, interpSpeedIncreasing=2}
    baseBoneRef {boneName=root}
    componentAppliedLinearAccClamp=<3000 3000 3000>
    componentLinearAccScale=<0.4 0.3 0.3>
    componentLinearVelScale=<0.4 0.3 0.3>
    simSpaceSettings {
     maxAngularAcceleration=1000
     maxAngularVelocity=3
     maxLinearAcceleration=700
     maxLinearVelocity=500
     velocityScaleZ=0
     worldAlpha=1
     externalAngularVelocity {z=0.1}
     externalLinearDragV=<0.2 0.2>
    }
   }
   n87 = RigidBody!(ComponentPose=@n86, Alpha, ExternalForce) -> (Pose) {
    bClampLinearTranslationLimitToRefPose=true
    bOverrideWorldGravity=true
    overridePhysicsAsset=&M_MED_StarWalkerFNCS_Collar_Physics
    simulationSpace=BaseBoneSpace
    alphaBoolBlend {blendOption=Linear}
    baseBoneRef {boneName=spine_05}
    componentAppliedLinearAccClamp=<1000 1000 1300>
    componentLinearAccScale=<0.8 0.8 0.6>
    componentLinearVelScale=<0.8 0.8 0.6>
    overrideWorldGravity {z=-950}
   }
   n88 = Get Is Playing Emote Float()
   n89 = "Transform (Modify) Bone - Bone: jacket_frt_01_r"(ComponentPose=@n87, Alpha=@n88, Translation="0,0,0", Rotation="-5,-7.5,0", Scale="1,1,1") {
    rotationMode=BMM_Additive
    rotationSpace=BCS_BoneSpace
    translationMode=BMM_Additive
    translationSpace=BCS_BoneSpace
    alphaBoolBlend {blendOption=Linear}
    alphaScaleBiasClamp {bInterpResult=true, interpSpeedIncreasing=2}
    boneToModify {boneName=jacket_frt_01_r}
    rotation {pitch=-5, yaw=-7.5}
   }
   n90 = "Transform (Modify) Bone - Bone: jacket_frt_01_l"(ComponentPose=@n89, Alpha=@n88, Translation="0,2,0", Rotation="5,-11,0", Scale="1,1,1") {
    rotationMode=BMM_Additive
    rotationSpace=BCS_BoneSpace
    translationMode=BMM_Additive
    translationSpace=BCS_BoneSpace
    alphaBoolBlend {blendOption=Linear}
    alphaScaleBiasClamp {bInterpResult=true, interpSpeedIncreasing=2}
    boneToModify {boneName=jacket_frt_01_l}
    rotation {pitch=5, yaw=-11}
    translation {y=2}
   }
   Output Pose(Result=Blend Poses by bool(
     BlendPose_0=Component To Local(ComponentPose=@n90)
     BlendPose_1=Blend Poses by bool(BlendPose_0=@n83, BlendPose_1=@n68, bActiveValue=@n84) {blendType=Linear}
     bActiveValue=Get bIsInFrontEnd()
    ) {blendType=Linear}) {layerGroup=None}
   n91 = "Constraint - jacket_frt_02_l (PPPPP)"!(
    ComponentPose=Local To Component(LocalPose=Copy Pose From Mesh() {bUseAttachedParent=true})
    Alpha=1
    ConstraintWeights_0=0.1633
    ConstraintWeights_1:float=0.3137
    ConstraintWeights_2:float=0.1434
    ConstraintWeights_3:float=0.3059
    ConstraintWeights_4:float=0.0737
   ) -> (Pose) {
    alphaBoolBlend {blendOption=Linear}
    boneToModify {boneName=jacket_frt_02_l}
    constraintSetup [
     {targetBone {boneName=spine_01}}
     {targetBone {boneName=spine_02}}
     {targetBone {boneName=spine_03}}
     {targetBone {boneName=spine_04}}
     {targetBone {boneName=pelvis}}
    ]
    constraintWeights [1,1,1,1,1]
   }
   n92 = "Constraint - jacket_frt_03_l (PPPPP)"(
    ComponentPose=@n91
    Alpha=1
    ConstraintWeights_0=0.1633
    ConstraintWeights_1=0.3137
    ConstraintWeights_2=0.1434
    ConstraintWeights_3=0.3059
    ConstraintWeights_4=0.0737
   ) {
    alphaBoolBlend {blendOption=Linear}
    boneToModify {boneName=jacket_frt_03_l}
    constraintSetup [
     {targetBone {boneName=spine_01}}
     {targetBone {boneName=spine_02}}
     {targetBone {boneName=spine_03}}
     {targetBone {boneName=spine_04}}
     {targetBone {boneName=pelvis}}
    ]
    constraintWeights [1,1,1,1,1]
   }
   n93 = "Constraint - jacket_frt_04_l (PP)"(ComponentPose=@n92, Alpha=1, ConstraintWeights_0=0.7617, ConstraintWeights_1=0.2383) {
    alphaBoolBlend {blendOption=Linear}
    boneToModify {boneName=jacket_frt_04_l}
    constraintSetup [{targetBone {boneName=pelvis}}, {targetBone {boneName=thigh_l}}]
    constraintWeights [1,1]
   }
   n94 = "Constraint - jacket_frt_02_r (PPPPP)"!(
    ComponentPose=@n93
    Alpha=1
    ConstraintWeights_0=0.1633
    ConstraintWeights_1:float=0.3137
    ConstraintWeights_2:float=0.1434
    ConstraintWeights_3:float=0.3059
    ConstraintWeights_4:float=0.0737
   ) -> (Pose) {
    alphaBoolBlend {blendOption=Linear}
    boneToModify {boneName=jacket_frt_02_r}
    constraintSetup [
     {targetBone {boneName=spine_01}}
     {targetBone {boneName=spine_02}}
     {targetBone {boneName=spine_03}}
     {targetBone {boneName=spine_04}}
     {targetBone {boneName=pelvis}}
    ]
    constraintWeights [1,1,1,1,1]
   }
   n95 = "Constraint - jacket_frt_03_r (PPPPP)"(
    ComponentPose=@n94
    Alpha=1
    ConstraintWeights_0=0.1633
    ConstraintWeights_1=0.3137
    ConstraintWeights_2=0.1434
    ConstraintWeights_3=0.3059
    ConstraintWeights_4=0.0737
   ) {
    alphaBoolBlend {blendOption=Linear}
    boneToModify {boneName=jacket_frt_03_r}
    constraintSetup [
     {targetBone {boneName=spine_01}}
     {targetBone {boneName=spine_02}}
     {targetBone {boneName=spine_03}}
     {targetBone {boneName=spine_04}}
     {targetBone {boneName=pelvis}}
    ]
    constraintWeights [1,1,1,1,1]
   }
   n96 = "Constraint - jacket_frt_04_r (PP)"(ComponentPose=@n95, Alpha=1, ConstraintWeights_0=0.2391, ConstraintWeights_1=0.7609) {
    alphaBoolBlend {blendOption=Linear}
    boneToModify {boneName=jacket_frt_04_r}
    constraintSetup [{targetBone {boneName=thigh_r}}, {targetBone {boneName=pelvis}}]
    constraintWeights [1,1]
   }
   n97 = "Constraint - jacket_frtSide_01_l (PPPPP)"!(
    ComponentPose=@n96
    Alpha=1
    ConstraintWeights_0=0.2331
    ConstraintWeights_1:float=0.3882
    ConstraintWeights_2:float=0.0505
    ConstraintWeights_3:float=0.1313
    ConstraintWeights_4:float=0.197
   ) -> (Pose) {
    alphaBoolBlend {blendOption=Linear}
    boneToModify {boneName=jacket_frtSide_01_l}
    constraintSetup [
     {targetBone {boneName=spine_01}}
     {targetBone {boneName=spine_02}}
     {targetBone {boneName=spine_03}}
     {targetBone {boneName=spine_04}}
     {targetBone {boneName=pelvis}}
    ]
    constraintWeights [1,1,1,1,1]
   }
   n98 = "Constraint - jacket_frtSide_02_l (PPPPP)"(
    ComponentPose=@n97
    Alpha=1
    ConstraintWeights_0=0.2331
    ConstraintWeights_1=0.3882
    ConstraintWeights_2=0.0505
    ConstraintWeights_3=0.1313
    ConstraintWeights_4=0.197
   ) {
    alphaBoolBlend {blendOption=Linear}
    boneToModify {boneName=jacket_frtSide_02_l}
    constraintSetup [
     {targetBone {boneName=spine_01}}
     {targetBone {boneName=spine_02}}
     {targetBone {boneName=spine_03}}
     {targetBone {boneName=spine_04}}
     {targetBone {boneName=pelvis}}
    ]
    constraintWeights [1,1,1,1,1]
   }
   n99 = "Constraint - jacket_frtSide_03_l (PP)"(ComponentPose=@n98, Alpha=1, ConstraintWeights_0=0.8376, ConstraintWeights_1=0.1624) {
    alphaBoolBlend {blendOption=Linear}
    boneToModify {boneName=jacket_frtSide_03_l}
    constraintSetup [{targetBone {boneName=pelvis}}, {targetBone {boneName=thigh_l}}]
    constraintWeights [1,1]
   }
   n100 = "Constraint - jacket_frtSide_01_r (PPPPP)"!(
    ComponentPose=@n99
    Alpha=1
    ConstraintWeights_0=0.2289
    ConstraintWeights_1:float=0.3815
    ConstraintWeights_2:float=0.0525
    ConstraintWeights_3:float=0.1321
    ConstraintWeights_4:float=0.2049
   ) -> (Pose) {
    alphaBoolBlend {blendOption=Linear}
    boneToModify {boneName=jacket_frtSide_01_r}
    constraintSetup [
     {targetBone {boneName=spine_01}}
     {targetBone {boneName=spine_02}}
     {targetBone {boneName=spine_03}}
     {targetBone {boneName=spine_04}}
     {targetBone {boneName=pelvis}}
    ]
    constraintWeights [1,1,1,1,1]
   }
   n101 = "Constraint - jacket_frtSide_02_r (PPPPP)"(
    ComponentPose=@n100
    Alpha=1
    ConstraintWeights_0=0.2289
    ConstraintWeights_1=0.3815
    ConstraintWeights_2=0.0525
    ConstraintWeights_3=0.1321
    ConstraintWeights_4=0.2049
   ) {
    alphaBoolBlend {blendOption=Linear}
    boneToModify {boneName=jacket_frtSide_02_r}
    constraintSetup [
     {targetBone {boneName=spine_01}}
     {targetBone {boneName=spine_02}}
     {targetBone {boneName=spine_03}}
     {targetBone {boneName=spine_04}}
     {targetBone {boneName=pelvis}}
    ]
    constraintWeights [1,1,1,1,1]
   }
   n102 = "Constraint - jacket_frtSide_03_r (PP)"(ComponentPose=@n101, Alpha=1, ConstraintWeights_0=0.1853, ConstraintWeights_1=0.8147) {
    alphaBoolBlend {blendOption=Linear}
    boneToModify {boneName=jacket_frtSide_03_r}
    constraintSetup [{targetBone {boneName=thigh_r}}, {targetBone {boneName=pelvis}}]
    constraintWeights [1,1]
   }
   n103 = "Constraint - jacket_side_01_l (PPPPP)"!(
    ComponentPose=@n102
    Alpha=1
    ConstraintWeights_0=0.2426
    ConstraintWeights_1:float=0.334
    ConstraintWeights_2:float=0.0344
    ConstraintWeights_3:float=0.0663
    ConstraintWeights_4:float=0.3227
   ) -> (Pose) {
    alphaBoolBlend {blendOption=Linear}
    boneToModify {boneName=jacket_side_01_l}
    constraintSetup [
     {targetBone {boneName=spine_01}}
     {targetBone {boneName=spine_02}}
     {targetBone {boneName=spine_03}}
     {targetBone {boneName=spine_04}}
     {targetBone {boneName=pelvis}}
    ]
    constraintWeights [1,1,1,1,1]
   }
   n104 = "Constraint - jacket_side_02_l (PPPPP)"(
    ComponentPose=@n103
    Alpha=1
    ConstraintWeights_0=0.2426
    ConstraintWeights_1=0.334
    ConstraintWeights_2=0.0344
    ConstraintWeights_3=0.0663
    ConstraintWeights_4=0.3227
   ) {
    alphaBoolBlend {blendOption=Linear}
    boneToModify {boneName=jacket_side_02_l}
    constraintSetup [
     {targetBone {boneName=spine_01}}
     {targetBone {boneName=spine_02}}
     {targetBone {boneName=spine_03}}
     {targetBone {boneName=spine_04}}
     {targetBone {boneName=pelvis}}
    ]
    constraintWeights [1,1,1,1,1]
   }
   n105 = "Constraint - jacket_side_03_l (PP)"(ComponentPose=@n104, Alpha=1, ConstraintWeights_0=0.8489, ConstraintWeights_1=0.1511) {
    alphaBoolBlend {blendOption=Linear}
    boneToModify {boneName=jacket_side_03_l}
    constraintSetup [{targetBone {boneName=pelvis}}, {targetBone {boneName=thigh_l}}]
    constraintWeights [1,1]
   }
   n106 = "Constraint - jacket_side_01_r (PPPPP)"!(
    ComponentPose=@n105
    Alpha=1
    ConstraintWeights_0=0.2507
    ConstraintWeights_1:float=0.3439
    ConstraintWeights_2:float=0.0296
    ConstraintWeights_3:float=0.0635
    ConstraintWeights_4:float=0.3122
   ) -> (Pose) {
    alphaBoolBlend {blendOption=Linear}
    boneToModify {boneName=jacket_side_01_r}
    constraintSetup [
     {targetBone {boneName=spine_01}}
     {targetBone {boneName=spine_02}}
     {targetBone {boneName=spine_03}}
     {targetBone {boneName=spine_04}}
     {targetBone {boneName=pelvis}}
    ]
    constraintWeights [1,1,1,1,1]
   }
   n107 = "Constraint - jacket_side_02_r (PPPPP)"(
    ComponentPose=@n106
    Alpha=1
    ConstraintWeights_0=0.2507
    ConstraintWeights_1=0.3439
    ConstraintWeights_2=0.0296
    ConstraintWeights_3=0.0635
    ConstraintWeights_4=0.3122
   ) {
    alphaBoolBlend {blendOption=Linear}
    boneToModify {boneName=jacket_side_02_r}
    constraintSetup [
     {targetBone {boneName=spine_01}}
     {targetBone {boneName=spine_02}}
     {targetBone {boneName=spine_03}}
     {targetBone {boneName=spine_04}}
     {targetBone {boneName=pelvis}}
    ]
    constraintWeights [1,1,1,1,1]
   }
   n108 = "Constraint - jacket_side_03_r (PP)"(ComponentPose=@n107, Alpha=1, ConstraintWeights_0=0.1511, ConstraintWeights_1=0.8489) {
    alphaBoolBlend {blendOption=Linear}
    boneToModify {boneName=jacket_side_03_r}
    constraintSetup [{targetBone {boneName=thigh_r}}, {targetBone {boneName=pelvis}}]
    constraintWeights [1,1]
   }
   n109 = "Constraint - jacket_back_01_l (PPPPP)"!(
    ComponentPose=@n108
    Alpha=1
    ConstraintWeights_0=0.2786
    ConstraintWeights_1:float=0.3325
    ConstraintWeights_2:float=0.0691
    ConstraintWeights_3:float=0.026
    ConstraintWeights_4:float=0.2939
   ) -> (Pose) {
    alphaBoolBlend {blendOption=Linear}
    boneToModify {boneName=jacket_back_01_l}
    constraintSetup [
     {targetBone {boneName=spine_01}}
     {targetBone {boneName=spine_02}}
     {targetBone {boneName=spine_03}}
     {targetBone {boneName=spine_04}}
     {targetBone {boneName=pelvis}}
    ]
    constraintWeights [1,1,1,1,1]
   }
   n110 = "Constraint - jacket_back_02_l (PPPPP)"(
    ComponentPose=@n109
    Alpha=1
    ConstraintWeights_0=0.2786
    ConstraintWeights_1=0.3325
    ConstraintWeights_2=0.0691
    ConstraintWeights_3=0.026
    ConstraintWeights_4=0.2939
   ) {
    alphaBoolBlend {blendOption=Linear}
    boneToModify {boneName=jacket_back_02_l}
    constraintSetup [
     {targetBone {boneName=spine_01}}
     {targetBone {boneName=spine_02}}
     {targetBone {boneName=spine_03}}
     {targetBone {boneName=spine_04}}
     {targetBone {boneName=pelvis}}
    ]
    constraintWeights [1,1,1,1,1]
   }
   n111 = "Constraint - jacket_back_03_l (PP)"(ComponentPose=@n110, Alpha=1, ConstraintWeights_0=0.9601, ConstraintWeights_1=0.0399) {
    alphaBoolBlend {blendOption=Linear}
    boneToModify {boneName=jacket_back_03_l}
    constraintSetup [{targetBone {boneName=pelvis}}, {targetBone {boneName=thigh_l}}]
    constraintWeights [1,1]
   }
   n112 = "Constraint - jacket_back_01_r (PPPPP)"!(
    ComponentPose=@n111
    Alpha=1
    ConstraintWeights_0=0.283
    ConstraintWeights_1:float=0.3464
    ConstraintWeights_2:float=0.072
    ConstraintWeights_3:float=0.0211
    ConstraintWeights_4:float=0.2775
   ) -> (Pose) {
    alphaBoolBlend {blendOption=Linear}
    boneToModify {boneName=jacket_back_01_r}
    constraintSetup [
     {targetBone {boneName=spine_01}}
     {targetBone {boneName=spine_02}}
     {targetBone {boneName=spine_03}}
     {targetBone {boneName=spine_04}}
     {targetBone {boneName=pelvis}}
    ]
    constraintWeights [1,1,1,1,1]
   }
   n113 = "Constraint - jacket_back_02_r (PPPPP)"(
    ComponentPose=@n112
    Alpha=1
    ConstraintWeights_0=0.283
    ConstraintWeights_1=0.3464
    ConstraintWeights_2=0.072
    ConstraintWeights_3=0.0211
    ConstraintWeights_4=0.2775
   ) {
    alphaBoolBlend {blendOption=Linear}
    boneToModify {boneName=jacket_back_02_r}
    constraintSetup [
     {targetBone {boneName=spine_01}}
     {targetBone {boneName=spine_02}}
     {targetBone {boneName=spine_03}}
     {targetBone {boneName=spine_04}}
     {targetBone {boneName=pelvis}}
    ]
    constraintWeights [1,1,1,1,1]
   }
   n114 = "Constraint - jacket_back_03_r (PP)"(ComponentPose=@n113, Alpha=1, ConstraintWeights_0=0.0399, ConstraintWeights_1=0.9601) {
    alphaBoolBlend {blendOption=Linear}
    boneToModify {boneName=jacket_back_03_r}
    constraintSetup [{targetBone {boneName=thigh_r}}, {targetBone {boneName=pelvis}}]
    constraintWeights [1,1]
   }
   n115 = "Constraint - jacket_back_01 (PPPPP)"!(
    ComponentPose=@n114
    Alpha=1
    ConstraintWeights_0=0.2641
    ConstraintWeights_1:float=0.3737
    ConstraintWeights_2:float=0.1147
    ConstraintWeights_3:float=0.0276
    ConstraintWeights_4:float=0.2199
   ) -> (Pose) {
    alphaBoolBlend {blendOption=Linear}
    boneToModify {boneName=jacket_back_01}
    constraintSetup [
     {targetBone {boneName=spine_01}}
     {targetBone {boneName=spine_02}}
     {targetBone {boneName=spine_03}}
     {targetBone {boneName=spine_04}}
     {targetBone {boneName=pelvis}}
    ]
    constraintWeights [1,1,1,1,1]
   }
   n116 = "Constraint - jacket_back_02 (PPPPP)"(
    ComponentPose=@n115
    Alpha=1
    ConstraintWeights_0=0.2641
    ConstraintWeights_1=0.3737
    ConstraintWeights_2=0.1147
    ConstraintWeights_3=0.0276
    ConstraintWeights_4=0.2199
   ) {
    alphaBoolBlend {blendOption=Linear}
    boneToModify {boneName=jacket_back_02}
    constraintSetup [
     {targetBone {boneName=spine_01}}
     {targetBone {boneName=spine_02}}
     {targetBone {boneName=spine_03}}
     {targetBone {boneName=spine_04}}
     {targetBone {boneName=pelvis}}
    ]
    constraintWeights [1,1,1,1,1]
   }
   n117 = "Constraint - jacket_back_03 (P)"(ComponentPose=@n116, Alpha=1, ConstraintWeights_0=1, ConstraintWeights_1=0) {
    alphaBoolBlend {blendOption=Linear}
    boneToModify {boneName=jacket_back_03}
    constraintSetup [{targetBone {boneName=pelvis}}]
    constraintWeights [1,0]
   }
   n118 = "Constraint - dyn_parent_shldr_tipA_l (PPPPPPPP)"!(
    ComponentPose=@n117
    Alpha=1
    ConstraintWeights_0=0.1879
    ConstraintWeights_1:float=0.1743
    ConstraintWeights_2:float=0.0866
    ConstraintWeights_3:float=0.0578
    ConstraintWeights_4:float=0.0894
    ConstraintWeights_5:float=0.2478
    ConstraintWeights_6:float=0.1352
    ConstraintWeights_7:float=0.0211
   ) -> (Pose) {
    alphaBoolBlend {blendOption=Linear}
    boneToModify {boneName=dyn_parent_shldr_tipA_l}
    constraintSetup [
     {targetBone {boneName=head}}
     {targetBone {boneName=clavicle_l}}
     {targetBone {boneName=upperarm_twist_02_l}}
     {targetBone {boneName=spine_03}}
     {targetBone {boneName=spine_04}}
     {targetBone {boneName=spine_05}}
     {targetBone {boneName=upperarm_twist_01_l}}
     {targetBone {boneName=neck_02}}
    ]
    constraintWeights [1,1,1,1,1,1,1,1]
   }
   n119 = "Constraint - dyn_parent_shldr_tipB_l (PPPPPPPP)"!(
    ComponentPose=@n118
    Alpha=1
    ConstraintWeights_0=0.1084
    ConstraintWeights_1:float=0.2337
    ConstraintWeights_2:float=0.1404
    ConstraintWeights_3:float=0.0548
    ConstraintWeights_4:float=0.0716
    ConstraintWeights_5:float=0.1791
    ConstraintWeights_6:float=0.1949
    ConstraintWeights_7:float=0.0173
   ) -> (Pose) {
    alphaBoolBlend {blendOption=Linear}
    boneToModify {boneName=dyn_parent_shldr_tipB_l}
    constraintSetup [
     {targetBone {boneName=head}}
     {targetBone {boneName=clavicle_l}}
     {targetBone {boneName=upperarm_twist_02_l}}
     {targetBone {boneName=spine_03}}
     {targetBone {boneName=spine_04}}
     {targetBone {boneName=spine_05}}
     {targetBone {boneName=upperarm_twist_01_l}}
     {targetBone {boneName=upperarm_l}}
    ]
    constraintWeights [1,1,1,1,1,1,1,1]
   }
   n120 = "Constraint - dyn_parent_cowl_l (PPPPPPPP)"!(
    ComponentPose=@n119
    Alpha=1
    ConstraintWeights_0=0.0439
    ConstraintWeights_1:float=0.2485
    ConstraintWeights_2:float=0.0943
    ConstraintWeights_3:float=0.0841
    ConstraintWeights_4:float=0.1478
    ConstraintWeights_5:float=0.1871
    ConstraintWeights_6:float=0.1077
    ConstraintWeights_7:float=0.0867
   ) -> (Pose) {
    alphaBoolBlend {blendOption=Linear}
    boneToModify {boneName=dyn_parent_cowl_l}
    constraintSetup [
     {targetBone {boneName=head}}
     {targetBone {boneName=clavicle_l}}
     {targetBone {boneName=upperarm_twist_02_l}}
     {targetBone {boneName=spine_02}}
     {targetBone {boneName=spine_03}}
     {targetBone {boneName=spine_04}}
     {targetBone {boneName=spine_05}}
     {targetBone {boneName=upperarm_twist_01_l}}
    ]
    constraintWeights [1,1,1,1,1,1,1,1]
   }
   n121 = "Constraint - dyn_parent_cowl_r (PPPPPPPP)"!(
    ComponentPose=@n120
    Alpha=1
    ConstraintWeights_0=0.0369
    ConstraintWeights_1:float=0.0852
    ConstraintWeights_2:float=0.0983
    ConstraintWeights_3:float=0.0863
    ConstraintWeights_4:float=0.1512
    ConstraintWeights_5:float=0.1898
    ConstraintWeights_6:float=0.1041
    ConstraintWeights_7:float=0.2482
   ) -> (Pose) {
    alphaBoolBlend {blendOption=Linear}
    boneToModify {boneName=dyn_parent_cowl_r}
    constraintSetup [
     {targetBone {boneName=head}}
     {targetBone {boneName=upperarm_twist_01_r}}
     {targetBone {boneName=upperarm_twist_02_r}}
     {targetBone {boneName=spine_02}}
     {targetBone {boneName=spine_03}}
     {targetBone {boneName=spine_04}}
     {targetBone {boneName=spine_05}}
     {targetBone {boneName=clavicle_r}}
    ]
    constraintWeights [1,1,1,1,1,1,1,1]
   }
   n122 = "Transform (Modify) Bone - Bone: jacket_frt_03_r"(ComponentPose=@n121, Alpha=1, Translation="0,3,0", Rotation="0,0,0", Scale="1,1,1") {
    rotationMode=BMM_Additive
    rotationSpace=BCS_BoneSpace
    translationMode=BMM_Additive
    translationSpace=BCS_BoneSpace
    alphaBoolBlend {blendOption=Linear}
    alphaScaleBiasClamp {bInterpResult=true, interpSpeedIncreasing=2}
    boneToModify {boneName=jacket_frt_03_r}
    translation {y=3}
   }
   n123 = "Transform (Modify) Bone - Bone: jacket_frt_03_l"(ComponentPose=@n122, Alpha=1, Translation="0,3,0", Rotation="0,0,0", Scale="1,1,1") {
    rotationMode=BMM_Additive
    rotationSpace=BCS_BoneSpace
    translationMode=BMM_Additive
    translationSpace=BCS_BoneSpace
    alphaBoolBlend {blendOption=Linear}
    alphaScaleBiasClamp {bInterpResult=true, interpSpeedIncreasing=2}
    boneToModify {boneName=jacket_frt_03_l}
    translation {y=3}
   }
   n124 = Get Is Crouching Float()
   n125 = "Transform (Modify) Bone - Bone: jacket_frt_02_r"(ComponentPose=@n123, Alpha=@n124, Translation="5,0,0", Rotation="10,0,0", Scale="1,1,1") {
    rotationMode=BMM_Additive
    rotationSpace=BCS_BoneSpace
    translationMode=BMM_Additive
    translationSpace=BCS_BoneSpace
    alphaBoolBlend {blendOption=Linear}
    alphaScaleBiasClamp {interpSpeedIncreasing=2}
    boneToModify {boneName=jacket_frt_02_r}
    rotation {pitch=10}
    translation {x=5}
   }
   n126 = "Transform (Modify) Bone - Bone: jacket_frt_02_l"(ComponentPose=@n125, Alpha=@n124, Translation="5,0,0", Rotation="10,0,0", Scale="1,1,1") {
    rotationMode=BMM_Additive
    rotationSpace=BCS_BoneSpace
    translationMode=BMM_Additive
    translationSpace=BCS_BoneSpace
    alphaBoolBlend {blendOption=Linear}
    alphaScaleBiasClamp {interpSpeedIncreasing=2}
    boneToModify {boneName=jacket_frt_02_l}
    rotation {pitch=10}
    translation {x=5}
   }
   n127 = "Transform (Modify) Bone - Bone: jacket_frtSide_02_l"(ComponentPose=@n126, Alpha=@n124, Translation="10,1,0", Rotation="0,0,0", Scale="1,1,1") {
    rotationMode=BMM_Additive
    rotationSpace=BCS_BoneSpace
    translationMode=BMM_Additive
    translationSpace=BCS_BoneSpace
    alphaBoolBlend {blendOption=Linear}
    alphaScaleBiasClamp {interpSpeedIncreasing=2}
    boneToModify {boneName=jacket_frtSide_02_l}
    translation=<10 1>
   }
   n128 = "Transform (Modify) Bone - Bone: jacket_frtSide_02_r"(ComponentPose=@n127, Alpha=@n124, Translation="10,1,0", Rotation="0,0,0", Scale="1,1,1") {
    rotationMode=BMM_Additive
    rotationSpace=BCS_BoneSpace
    translationMode=BMM_Additive
    translationSpace=BCS_BoneSpace
    alphaBoolBlend {blendOption=Linear}
    alphaScaleBiasClamp {interpSpeedIncreasing=2}
    boneToModify {boneName=jacket_frtSide_02_r}
    translation=<10 1>
   }
   n129 = "Transform (Modify) Bone - Bone: jacket_side_02_l"(ComponentPose=@n128, Alpha=@n124, Translation="1,0,0", Rotation="0,0,0", Scale="1,1,1") {
    rotationMode=BMM_Additive
    rotationSpace=BCS_BoneSpace
    translationMode=BMM_Additive
    translationSpace=BCS_BoneSpace
    alphaBoolBlend {blendOption=Linear}
    alphaScaleBiasClamp {interpSpeedIncreasing=2}
    boneToModify {boneName=jacket_side_02_l}
    translation {x=1}
   }
   n130 = "Transform (Modify) Bone - Bone: jacket_side_02_r"(ComponentPose=@n129, Alpha=@n124, Translation="1,0,0", Rotation="0,0,0", Scale="1,1,1") {
    rotationMode=BMM_Additive
    rotationSpace=BCS_BoneSpace
    translationMode=BMM_Additive
    translationSpace=BCS_BoneSpace
    alphaBoolBlend {blendOption=Linear}
    alphaScaleBiasClamp {interpSpeedIncreasing=2}
    boneToModify {boneName=jacket_side_02_r}
    translation {x=1}
   }
   Save cached pose 'AnimCache'(Pose=Component To Local(ComponentPose=@n130)) {cacheName=AnimCache}
  ]}
 EventGraph {nodes [
   Event BlueprintUpdateAnimation()
   > Get!(execute:Exec) -> (then:Exec, else:Exec, Skeletal_Mesh:Skeletal Mesh Component Object Reference)
   > n1 = Sequence!(execute) -> (then_0, then_1, then_2:Exec, then_3:Exec, then_4:Exec, then_5:Exec, then_6:Exec, then_7:Exec, then_8:Exec)
   n2 = Get bIsDBNOCarried()
   Set Is DBNOCarried(execute=@n1.then_4, Is DBNOCarried:double="To Float (Boolean)"(InBool=@n2)) -> (Output_Get:double)
   > Set Is Not DBNOCarried(Is Not DBNOCarried:double="To Float (Boolean)"(InBool=NOT Boolean(A=@n2))) -> (Output_Get:double)
   n3 = Sequence!(execute=@n1.then_7) -> (then_0, then_1, then_2, then_3, then_4, then_5)
   > n4 = Gravity Override State Function DEPRECATED(CharacterPartAnimInstance=Self-Reference(), GameTimeInSeconds=GetGameTimeInSeconds(), GravityOverrideParams=Get Gravity Override Params())
   > Set GravityOverride(GravityOverride=@n4)
   Set CoatAlpha(
    execute=@n3.then_1
    CoatAlpha:double=float * float!(
     A=Get Is Sky Diving Coat Alpha() -> (Is Sky Diving Coat Alpha:double)
     B=Get Is Tactical Sprinting Alpha() -> (Is Tactical Sprinting Alpha:double)
     C:float=Get Is Sliding Alpha() -> (Is Sliding Alpha:float)
     D:double=Get Jumping Coat Alpha() -> (Jumping Coat Alpha:double)
     E:double=Get Is Swim Sprinting Alpha() -> (Is Swim Sprinting Alpha:double)
     F:double=Get Is Surface Swimming Alpha() -> (Is Surface Swimming Alpha:double)
     G:double=Get Is Not DBNOCarried() -> (Is Not DBNOCarried:double)
     H:float=Get Is Crouching Alpha() -> (Is Crouching Alpha:float)
     I:double=Get Is playing Emote Alpha() -> (Is playing Emote Alpha:double)
    ) -> (ReturnValue) {is=title K2Node_PromotableOperator}
   ) -> (Output_Get:double)
   Set Is Crouching Float(execute=@n3.then_2, Is Crouching Float:double="To Float (Boolean)"(InBool=Get bIsCrouching())) -> (Output_Get:double)
   n5 = Sequence!(execute=@n3.then_3) -> (then_0, then_1, then_2, then_3, then_4, then_5, then_6:Exec, then_7:Exec)
   > n6 = Branch(Condition=Get bIsSkyDiving())
   > Set Is Sky Diving Coat Alpha(Is Sky Diving Coat Alpha:double=1) -> (Output_Get:double)
   Set Is Sky Diving Coat Alpha(execute=@n6.else, Is Sky Diving Coat Alpha=1)
   n7 = Branch(execute=@n5.then_1, Condition=Get bIsTacticalSprinting())
   > Set Is Tactical Sprinting Alpha(Is Tactical Sprinting Alpha:double=0.8) -> (Output_Get:double)
   Set Is Tactical Sprinting Alpha(execute=@n7.else, Is Tactical Sprinting Alpha=1)
   n8 = Branch(execute=@n5.then_2, Condition=Get bIsSliding())
   > Set Is Sliding Alpha(Is Sliding Alpha:float=0.5) -> (Output_Get:float)
   Set Is Sliding Alpha(execute=@n8.else, Is Sliding Alpha=1)
   n9 = NOT Boolean(A=Get bIsSkyDiving())
   n10 = Get bIsJumping()
   n11 = NOT Boolean(A=Get bIsSkyDiving())
   n12 = Get bIsFalling()
   n13 = Branch(execute=@n5.then_3, Condition=OR Boolean(A=AND Boolean(A=@n10, B=@n9), B=AND Boolean(A=@n12, B=@n11)))
   > Set Jumping Coat Alpha(Jumping Coat Alpha:double=1) -> (Output_Get:double)
   Set Jumping Coat Alpha(execute=@n13.else, Jumping Coat Alpha=1)
   n14 = Branch(execute=@n5.then_4, Condition=Get bIsSurfaceSwimming())
   > Set Is Surface Swimming Alpha(Is Surface Swimming Alpha:double=1) -> (Output_Get:double)
   Set Is Surface Swimming Alpha(execute=@n14.else, Is Surface Swimming Alpha=1)
   n15 = Branch(execute=@n5.then_5, Condition=Get bIsSwimSprinting() -> (bIsSwimSprinting:bool))
   > Set Is Swim Sprinting Alpha(Is Swim Sprinting Alpha:double=1) -> (Output_Get:double)
   Set Is Swim Sprinting Alpha(execute=@n15.else, Is Swim Sprinting Alpha=1)
   n16 = Branch(execute=@n5.then_6, Condition=Get bIsCrouching())
   > Set Is Crouching Alpha(Is Crouching Alpha:float=1) -> (Output_Get:float)
   Set Is Crouching Alpha(execute=@n16.else, Is Crouching Alpha=1)
   n17 = Branch(execute=@n5.then_7, Condition=Get bIsPlayingEmote())
   > Set Is playing Emote Alpha(Is playing Emote Alpha:double=0.5) -> (Output_Get:double)
   Set Is playing Emote Alpha(execute=@n17.else, Is playing Emote Alpha=1)
   Set Is Playing Emote Float(execute=@n3.then_5, Is Playing Emote Float:double="To Float (Boolean)"(InBool=Get bIsPlayingEmote())) -> (Output_Get:double)
   n18 = Sequence!(execute=@n3.then_4) -> (then_0, then_1, then_2, then_3, then_4)
   > n19 = Cast To FortPlayerAnimInstance(Object=GetAnimInstance(self=GetOwnerSkeletalMeshComponent()))
   > Set isPistol(isPistol=false)
   > Set isRifle(isRifle:bool=false) -> (Output_Get:bool)
   > n20 = Switch on EFortWeaponCoreAnimation!(execute, Selection:EFortWeaponCoreAnimation Enum=Get WeaponCoreAnim(self=@n19)) -> (
    Melee:Exec
    Pistol:Exec
    Shotgun:Exec
    PaperBlueprint:Exec
    Rifle:Exec
    MeleeOneHand:Exec
    MachinePistol:Exec
    RocketLauncher:Exec
    GrenadeLauncher:Exec
    GoingCommando:Exec
    AssaultRifle:Exec
    TacticalShotgun:Exec
    SniperRifle:Exec
    TrapPlacement:Exec
    ShoulderLauncher:Exec
    AbilityDecoTool:Exec
    Crossbow:Exec
    C4:Exec
    RemoteControl:Exec
    DualWield:Exec
    AR_BullPup:Exec
    AR_ForwardGrip:Exec
    MedPackPaddles:Exec
    SMG_P90:Exec
    AR_DrumGun:Exec
    Consumable_Small:Exec
    Consumable_Large:Exec
    Balloon:Exec
    MountedTurret:Exec
    CreativeTool:Exec
    ExplosiveBow:Exec
    AshtonIndigo:Exec
    AshtonChicago:Exec
    MeleeDualWield:Exec
    Unarmed:Exec
   )
   Set isRifle(execute=[@n20.AssaultRifle, @n20.Rifle, @n20.RocketLauncher, @n20.Shotgun, @n20.SniperRifle, @n20.TacticalShotgun], isRifle=true)
   Set isPistol(execute=@n20.Pistol, isPistol=true)
   Set Is Rifle Alpha(execute=@n18.then_1, Is Rifle Alpha:double="To Float (Boolean)"(InBool=Get isRifle() -> (isRifle:bool))) -> (Output_Get:double)
   Set Is Pistol Alpha(execute=@n18.then_2, Is Pistol Alpha:double="To Float (Boolean)"(InBool=Get isPistol() -> (isPistol:bool))) -> (Output_Get:double)
   n21 = Get bIsTargeting()
   n22 = Get isRifle()
   Set Is Rifle Targeting Alpha(execute=@n18.then_3, Is Rifle Targeting Alpha:double="To Float (Boolean)"(InBool=AND Boolean(A=@n22, B=@n21))) -> (Output_Get:double)
   n23 = Get bIsTargeting()
   n24 = Get isPistol()
   Set Is Pistol Targeting Alpha(execute=@n18.then_4, Is Pistol Targeting Alpha:float="To Float (Boolean)"(InBool=AND Boolean(A=@n24, B=@n23))) -> (Output_Get:float)
   n25 = Branch(execute=@n1.then_0, Condition=Get bIsSkyDiving())
   n26 = Break Fort Anim Input Skydiving External Force(FortAnimInput_SkydivingExternalForce=Get SkydivingExternalForceInput()) -> (
    bUseSkydivingVectorForce=~
    SkydivingVectorForce=~
    HeadToPelvisDirection=~
    FloatingMultiplier=~
    FloatingAdditive=~
    DivingMultiplier=~
    DivingAdditive=~
    ParachutingMultiplier=~
    ParachutingAdditive=~
   )
   > Set skydiving_force(skydiving_force:Vector=vector * float(A=@n26.SkydivingVectorForce, B=0.65)) -> (Output_Get:Vector)
   > Branch(Condition=Get bIsParachuteOpened())
   > Set skydiving_force(skydiving_force=vector * float(A=@n26.SkydivingVectorForce, B=0.1))
   Set skydiving_force(execute=@n25.else, skydiving_force="0, 0, 0")
   n27 = BreakTransform(InTransform=GetDeltaTransformFromRefPose(self=Get Skeletal_Mesh() -> (Skeletal_Mesh:Skeletal Mesh Component Object Reference), BoneName=calf_l, BaseName=thigh_l))
   Set calf_rot_back_l(execute=@n1.then_3, calf_rot_back_l:double=MapRangeClamped(Value=BreakRotator(InRot=@n27.Rotation).Yaw, InRangeA=15, InRangeB=115)) -> (Output_Get:double)
   n28 = Get bRigidBodyEnabled()
   n29 = Get CurrentLOD()
   Set rigids_off(execute=@n1.then_1, rigids_off:bool=OR Boolean(A=NOT Boolean(A=@n28), B="integer > integer"(A=@n29, B=1) {is=title K2Node_CallFunction})) -> (Output_Get:bool)
   Branch(execute=@n1.then_6, Condition=Get fromDBNO() -> (fromDBNO:bool))
   > Branch(Condition=NOT Boolean(A=Get bIsDBNO()))
   > Delay(Duration=0.15)
   > ResetDynamics(InTeleportType=TeleportPhysics)
   > ResetDynamics(InTeleportType=ResetPhysics)
   > n30 = Set fromDBNO(fromDBNO:bool=false) -> (Output_Get:bool)
   Do Once(execute=@n31.then, Reset=@n30.then)
   > Set fromDBNO(fromDBNO=true)
   n31 = Branch(execute=@n1.then_5, Condition=Get bIsDBNO())
   n32 = BreakTransform(InTransform=GetDeltaTransformFromRefPose(self=Get Skeletal_Mesh(), BoneName=calf_r, BaseName=thigh_r))
   Set calf_rot_back_r(execute=@n1.then_2, calf_rot_back_r:double=MapRangeClamped(Value=BreakRotator(InRot=@n32.Rotation).Yaw, InRangeA=15, InRangeB=115)) -> (Output_Get:double)
   Event OnParachuteOpened()
   > n33 = Delay(Duration=1.5)
   ResetDynamics(execute=[@n33.then, @n34.then, @n35.then], InTeleportType=TeleportPhysics)
   > ResetDynamics(InTeleportType=ResetPhysics)
   n34 = Event OnEndSkydiving()
   Event BlueprintInitializeAnimation()
   > ResetDynamics(InTeleportType=TeleportPhysics)
   > Set Skeletal_Mesh(Skeletal_Mesh:Skeletal Mesh Component Object Reference=GetOwnerSkeletalMeshComponent()) -> (Output_Get:Skeletal Mesh Component Object Reference)
   n35 = Event OnBeginSkydiving()
  ]}
]
variables [
 calf_rot_back_l {type=number}
 calf_rot_back_r {type=number}
 CoatAlpha {type=number}
 fromDBNO {type=boolean}
 Gravity Override Params {type=GravityOverrideParamsStruct}
 GravityOverride {type=Vector}
 Is Crouching Alpha {type=number}
 Is Crouching Float {type=number}
 Is DBNOCarried {type=number}
 Is Not DBNOCarried {type=number}
 Is Pistol Alpha {type=number}
 Is Pistol Targeting Alpha {type=number}
 Is playing Emote Alpha {type=number}
 Is Playing Emote Float {type=number}
 Is Rifle Alpha {type=number}
 Is Rifle Targeting Alpha {type=number}
 Is Sky Diving Coat Alpha {type=number}
 Is Sliding Alpha {type=number}
 Is Surface Swimming Alpha {type=number}
 Is Swim Sprinting Alpha {type=number}
 Is Tactical Sprinting Alpha {type=number}
 isPistol {type=boolean}
 isRifle {type=boolean}
 Jumping Coat Alpha {type=number}
 rigids_off {type=boolean}
 Skeletal_Mesh {type=SkeletalMeshComponent}
 skydiving_force {type=Vector}
]
